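# Niagara system for the Mythical skydiving contrail cosmetic, trailing paired orange and blue effects from the character's skeletal mesh.
# BodyTrails spawns mesh particles at sampled points, while FinTrail002 and Main001 lay burst-spawned particles into partitioned ribbons by normalized index, with waveform offsets.
# ParaGlideLeanAlpha and FN_Contrail_Setup's dive value drive length, width or brightness, alongside front-end and time-of-day scaling.
# The kite and gizmo emitters are disabled.
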NiagaraSystem /BRCosmetics/Athena/Effects/Particles/Skydiving/NS_Contrail_Mythical/NS_Contrail_Mythical
dependencies {
 usedDataInterfaces [
  &NiagaraDataInterfaceArrayColor
  &NiagaraDataInterfaceCurve
  &NiagaraDataInterfaceVectorCurve
  &NiagaraDataInterfaceArrayFloat3
  &NiagaraDataInterfaceArrayInt32
  &NiagaraDataInterfacePlatformSet
  &NiagaraDataInterfaceMeshRendererInfo
  &NiagaraDataInterfaceArrayFloat
  &NiagaraDataInterfaceColorCurve
 ]
 usedDynamicInputs [&MakeQuatFromAxisAngle, &MaskFloatBySpawnGroup]
 usedModules [
  &self:SystemSpawnScript.SystemScriptSource.SystemScriptGraph.NiagaraNodeAssignment_0.SetVariables_#1
  &self:SystemSpawnScript.SystemScriptSource.SystemScriptGraph.NiagaraNodeAssignment_1.SetVariables_#2
  &SystemState
  /Game/Effects/Niagara/Modules/Contrails/FN_Contrail_Setup
  &FN_Integrator
 ]
}
emitters {
 BodyTrails {
  emitterUpdateScript [
   &EmitterState(
    Inactive Response=::NewEnumerator1
    Life Cycle Mode=::NewEnumerator1
    Loop Behavior=::NewEnumerator1
    Loop Duration Mode=::NewEnumerator1
    UseLoopDelay=false
   )
   &SpawnRate(
    SpawnRate=&MakeCustomFloatFromBool(
     Bool=&SetBoolByFloatComparison(A=1, B=$User.ParaGlideLeanAlpha, Comparison Type=::NewEnumerator4)
     True Float=12
     False Float=6
    )
    Use Spawn Probability=false
   )
   /SetVariables(Emitter.Width:NiagaraFloat=&MakeCustomFloatFromBool(
     Bool=&SetBoolByFloatComparison(A=1, B=$User.ParaGlideLeanAlpha, Comparison Type=::NewEnumerator4)
     True Float=25
     False Float=15
    )) {isSetParameters=true}
  ]
  particleSpawnScript [
   &PartitionParticles()
   &InitializeParticle(
    Color=&MaskLinearColor(
     Mask Source=1
     Mask Parameter=$Particles.PartitionParticles.Partition
     Mask=0
     Masked Linear Color=<rgba 0.724 0.0879 0 1>
     Passthrough Linear Color=<rgba 0 0.0701 0.7083 1>
    )
    Lifetime=&MakeCustomFloatFromBool(
     Bool=&SetBoolByFloatComparison(A=1, B=$User.ParaGlideLeanAlpha, Comparison Type=::NewEnumerator4)
     True Float=0.7
     False Float=1.25
    )
    Mesh Scale Max=<2 1.5 1.5>
    Mesh Scale Min=<1.2 1 1>
    Mesh Scale Mode=::NewEnumerator2
    UsePositionOffset=false
    Write Material Random=true
   )
   &InitialMeshOrientation(Mesh Orientation Mode=::NewEnumerator2, Rotation=<0 0.25 0>, Use Rotation Vector=true)
   &InitialMeshOrientation(
    Mesh Orientation Mode=::NewEnumerator2
    Rotation=&RandomRangeVector(Minimum=<0 0 -1>, Maximum=<0 0 1>)
    Use Rotation Vector=true
   ) {module=InitialMeshOrientation001}
   &ShapeLocation(
    Cylinder Height=25
    Cylinder Radius=5
    MaskSpawnGroup=true
    Override Seed=false
    Rotation Mode=::NewEnumerator1
    Shape Primitive=::NewEnumerator1
    Surface Only=false
    UseOwnerScale=true
   )
   &AddVelocity(
    Velocity=&MakeCustomVectorFromBool(
     Bool=&SetBoolByFloatComparison(A=1, B=$User.ParaGlideLeanAlpha, Comparison Type=::NewEnumerator4)
     True Vector=<0 0 80>
     False Vector=<1 0 30>
    )
    Velocity Speed Scale=0
   )
  ]
  particleUpdateScript [
   &ParticleState()
   &ScaleMeshSize(Scale Factor=&Lerp_Vector(A=<1 1 1>, B=<1.5 0.7 0.7>, Alpha=$User.ParaGlideLeanAlpha))
   &SkeletalMeshLocation(
    Bone / Socket Index=$Particles.PartitionParticles.Partition
    MaskSpawnGroup=true
    Mesh Sampling Type=::NewEnumerator5
    Orientation Sampling=::NewEnumerator1
    Position Sampling=::NewEnumerator1
    Sampled Position Offset=&AddVectorToPosition(
     Position=Linked() {linkedVariable {name=Particles.Initial.ShapeLocation.ShapePosition, type=Vector3f}}
     Vector=&SelectVectorFromArray(
      Direct Array Index=$Particles.PartitionParticles.Partition
      Vector Selection Array=DataInterface() {propertyValues {FloatData [<0 0 0>, <0 0 0>]}}
     )
    )
    Sampled Position Offset Coordinate Space:ENiagaraSkelMeshTransforms=::NewEnumerator2
    Skeletal Mesh=$System.Skeletal Mesh
    Socket Sampling Mode=::NewEnumerator5
   )
   &DynamicMaterialParameters(
    Index 0 Param 3=1
    Index 0 Param 4=&FloatFromCurve(FloatCurve=DataInterface() {propertyValues {Curve {keys [<at 0 0>,<at 1 1>]}}})
    Index 3 Param 1=500
    Module.Index0Param0Write=true
    Module.Index0Param1Write=true
    Module.Index0Param2Write=true
    Module.Index0Param3Write=true
    Module.Index3Param0Write=true
    Module.Index3Param1Write=true
    Module.Index3Param2Write=true
    Module.Index3Param3Write=true
    Write Parameter Index 3=true
   )
   &ScaleColor(
    Scale Alpha=&FloatFromCurve(
     FloatCurve=DataInterface() {propertyValues {Curve {keys [
         {interpMode=RCIM_Cubic, tangentMode=RCTM_User, time=0, value=0}
         {interpMode=RCIM_Cubic, tangentMode=RCTM_User, time=0.4, value=0.8}
         {tangentMode=RCTM_User, time=1, value=0}
        ]}}}
     Scale Curve=&Multiply_Float(
      A=&Lerp_Float(A=0.75, B=1, Alpha=$User.ParaGlideLeanAlpha)
      B=&FN_FrontEndFloatModifier(Gameplay=1, Front End=0.3)
     )
    )
    Scale RGB=<2 2 2>
    ScaleA=true
    ScaleRGB=true
   )
   &SolveForcesAndVelocity(Clamp Velocity=false, Limit Acceleration=false)
   &UpdateMeshOrientation(
    Facing Direction=$System.FN_Contrail_Setup.Direction Vector
    Facing Interpolation=false
    Orientation Method=::NewEnumerator1
    Override Quat=false
    TransformFacing=false
    TransformUp=false
    Up Direction:Vector3f=<1 0 0>
   ) {enabled=false}
  ]
  rendererClasses [&NiagaraMeshRendererProperties]
  renderers [{rendererClass=&NiagaraMeshRendererProperties, rendererIndex=0}]
 }
 FN_Contrails_Kite_v2 {
  bEnabled=false
  emitterUpdateScript [
   &EmitterScalability(
    Module.bApplySpawnCountScale=false
    Module.bApplySpawnCountScaleByDistanceFraction=false
    Module.bUseMaxDistance=false
    Module.bUseMinDistance=false
   ) {enabled=false}
   &EmitterLifeCycle(NextLoopDuration=1.9958) {enabled=false}
   &EmitterState()
   /Game/Effects/Niagara/Modules/Spawn/FN_Spawn_Kite(
    "Link Per Kite  Amount"=1
    Noise=DataInterface() {propertyValues {
      XCurve {keys [
        <at 0 0 RCIM_Cubic>
        {arriveTangent=-22.177, interpMode=RCIM_Cubic, leaveTangent=-22.177, time=0.0148, value=0.3241}
        {arriveTangent=8.0832, interpMode=RCIM_Cubic, leaveTangent=8.0832, time=0.0653, value=-1.4471}
        {arriveTangent=39.4964, interpMode=RCIM_Cubic, leaveTangent=39.4964, time=0.1184, value=1.1614}
        <at 0.1293 1.083>
        {arriveTangent=-23.4377, interpMode=RCIM_Cubic, leaveTangent=-23.4377, time=0.1576, value=0.0233}
        {arriveTangent=-0.0218, interpMode=RCIM_Cubic, leaveTangent=-0.0218, time=0.2029, value=-0.642}
        <at 0.267 0.0209>
        {arriveTangent=-9.845, interpMode=RCIM_Cubic, leaveTangent=-9.845, time=0.2888, value=0.1309}
        {arriveTangent=3.938, interpMode=RCIM_Cubic, leaveTangent=3.938, time=0.3474, value=-0.7708}
        {arriveTangent=2.5316, interpMode=RCIM_Cubic, leaveTangent=2.5316, time=0.3869, value=0.5173}
        {arriveTangent=-4.044, interpMode=RCIM_Cubic, leaveTangent=-4.044, time=0.4237, value=-0.5776}
        <at 0.4469 0.2748>
        {arriveTangent=-4.0955, interpMode=RCIM_Cubic, leaveTangent=-4.0955, time=0.4551, value=0.4851}
        {arriveTangent=-8.4961, interpMode=RCIM_Cubic, leaveTangent=-8.4961, time=0.5107, value=0.0134}
        <at 0.5109 0.0104>
        {arriveTangent=-0.6728, interpMode=RCIM_Cubic, leaveTangent=-0.6728, time=0.5736, value=-1.125}
        <at 0.6118 -0.0575>
        {arriveTangent=17.1461, interpMode=RCIM_Cubic, leaveTangent=17.1461, time=0.6146, value=0.0105}
        {arriveTangent=11.3424, interpMode=RCIM_Cubic, leaveTangent=11.3424, time=0.6622, value=0.8072}
        {arriveTangent=-13.2445, interpMode=RCIM_Cubic, leaveTangent=-13.2445, time=0.699, value=0.9682}
        {arriveTangent=-25.0454, interpMode=RCIM_Cubic, leaveTangent=-25.0454, time=0.7226, value=0.0074}
        {arriveTangent=-0.028, interpMode=RCIM_Cubic, leaveTangent=-0.028, time=0.7672, value=-0.7386}
        {arriveTangent=21.9983, interpMode=RCIM_Cubic, leaveTangent=21.9983, time=0.7974, value=0.0053}
        {arriveTangent=5.888, interpMode=RCIM_Cubic, leaveTangent=5.888, time=0.8462, value=1.0004}
        {arriveTangent=-15.155, interpMode=RCIM_Cubic, leaveTangent=-15.155, time=0.8953, value=0.5818}
        <at 0.9119 0.0053>
        {arriveTangent=-7.8259, interpMode=RCIM_Cubic, leaveTangent=-7.8259, time=0.9321, value=-0.6098}
        <at 0.9591 -0.3645>
        {arriveTangent=8.9177, interpMode=RCIM_Cubic, leaveTangent=8.9177, time=0.9879, value=0}
        {arriveTangent=0.682, interpMode=RCIM_Cubic, leaveTangent=0.682, tangentMode=RCTM_User, time=1, value=0}
       ]}
      YCurve {keys [
        <at 0 0 RCIM_Cubic>
        {arriveTangent=0.0179, interpMode=RCIM_Cubic, leaveTangent=0.0179, time=0.0843, value=1.1292}
        {arriveTangent=-19.1001, interpMode=RCIM_Cubic, leaveTangent=-19.1001, time=0.117, value=0.0021}
        <at 0.1293 0.2701>
        {arriveTangent=-34.1213, interpMode=RCIM_Cubic, leaveTangent=-34.1213, time=0.1607, value=1.1225}
        {arriveTangent=-41.5782, interpMode=RCIM_Cubic, leaveTangent=-41.5782, time=0.1702, value=-1.125}
        {arriveTangent=2.3241, interpMode=RCIM_Cubic, leaveTangent=2.3241, time=0.2015, value=-0.5776}
        {arriveTangent=-2.7223, interpMode=RCIM_Cubic, leaveTangent=-2.7223, time=0.2533, value=-0.9318}
        <at 0.267 -0.7557>
        {arriveTangent=19.4781, interpMode=RCIM_Cubic, leaveTangent=19.4781, time=0.3065, value=0.1309}
        {arriveTangent=-8.6722, interpMode=RCIM_Cubic, leaveTangent=-8.6722, time=0.3637, value=1.1292}
        <at 0.4469 -1.0865>
        {arriveTangent=41.2205, interpMode=RCIM_Cubic, leaveTangent=41.2205, time=0.4687, value=-1.3504}
        {arriveTangent=49.9749, interpMode=RCIM_Cubic, leaveTangent=49.9749, time=0.4959, value=0.936}
        <at 0.5109 0.7611>
        {arriveTangent=3.1016, interpMode=RCIM_Cubic, leaveTangent=3.1016, time=0.5532, value=-0.5454}
        {arriveTangent=9.5865, interpMode=RCIM_Cubic, leaveTangent=9.5865, time=0.5777, value=0.9682}
        <at 0.6118 0.0165>
        {arriveTangent=11.842, interpMode=RCIM_Cubic, leaveTangent=11.842, time=0.6309, value=-0.4165}
        {arriveTangent=-2.9535, interpMode=RCIM_Cubic, leaveTangent=-2.9535, time=0.6731, value=0.7428}
        {arriveTangent=-5.1686, interpMode=RCIM_Cubic, leaveTangent=-5.1686, time=0.6854, value=-0.5776}
        {arriveTangent=7.3333, interpMode=RCIM_Cubic, leaveTangent=7.3333, time=0.7167, value=0.5173}
        {arriveTangent=-15.9313, interpMode=RCIM_Cubic, leaveTangent=-15.9313, time=0.7641, value=0}
        <at 0.8095 -0.9606>
        {interpMode=RCIM_Cubic, tangentMode=RCTM_User, time=0.8489, value=0}
        {interpMode=RCIM_Cubic, tangentMode=RCTM_User, time=0.904, value=0.6415}
        {interpMode=RCIM_Cubic, tangentMode=RCTM_User, time=0.9329, value=0.2589}
        {interpMode=RCIM_Cubic, tangentMode=RCTM_User, time=0.967, value=-0.5063}
        {interpMode=RCIM_Cubic, tangentMode=RCTM_User, time=1, value=0}
       ]}
      ZCurve {keys [
        {interpMode=RCIM_Cubic, tangentMode=RCTM_User, time=0, value=0}
        {arriveTangent=10.357, interpMode=RCIM_Cubic, leaveTangent=10.357, tangentMode=RCTM_User, time=0.0489, value=-1.0284}
        {arriveTangent=5.6189, interpMode=RCIM_Cubic, leaveTangent=5.6189, tangentMode=RCTM_User, time=0.0854, value=0.8726}
        {arriveTangent=-0.1676, interpMode=RCIM_Cubic, leaveTangent=-0.1676, tangentMode=RCTM_User, time=0.1245, value=-0.6036}
        {arriveTangent=-0.5858, interpMode=RCIM_Cubic, leaveTangent=-0.5858, tangentMode=RCTM_User, time=0.1956, value=0.8542}
        {interpMode=RCIM_Cubic, tangentMode=RCTM_User, time=0.219, value=-0.659}
        {interpMode=RCIM_Cubic, tangentMode=RCTM_User, time=0.2776, value=0.8542}
        {arriveTangent=-3.0163, interpMode=RCIM_Cubic, leaveTangent=-3.0163, tangentMode=RCTM_User, time=0.2916, value=-0.659}
        {arriveTangent=7.8759, interpMode=RCIM_Cubic, leaveTangent=7.8759, tangentMode=RCTM_User, time=0.3877, value=0.522}
        {arriveTangent=-10.0071, interpMode=RCIM_Cubic, leaveTangent=-10.0071, tangentMode=RCTM_User, time=0.4439, value=0.5405}
        {arriveTangent=-1.5965, interpMode=RCIM_Cubic, leaveTangent=-1.5965, tangentMode=RCTM_User, time=0.5205, value=-0.8066}
        {arriveTangent=1.6215, interpMode=RCIM_Cubic, leaveTangent=1.6215, tangentMode=RCTM_User, time=0.5595, value=0.3559}
        {arriveTangent=2.5933, interpMode=RCIM_Cubic, leaveTangent=2.5933, tangentMode=RCTM_User, time=0.6001, value=-0.6774}
        {arriveTangent=26.7367, interpMode=RCIM_Cubic, leaveTangent=26.7367, tangentMode=RCTM_User, time=0.6235, value=0.522}
        {arriveTangent=-16.0548, interpMode=RCIM_Cubic, leaveTangent=-16.0548, tangentMode=RCTM_User, time=0.6595, value=0.9095}
        {arriveTangent=-7.1505, interpMode=RCIM_Cubic, leaveTangent=-7.1505, tangentMode=RCTM_User, time=0.6845, value=-0.456}
        {arriveTangent=5.7226, interpMode=RCIM_Cubic, leaveTangent=5.7226, tangentMode=RCTM_User, time=0.7188, value=0.4851}
        {interpMode=RCIM_Cubic, tangentMode=RCTM_User, time=0.7641, value=0}
        <at 0.8148 -0.6258>
        {arriveTangent=6.4436, interpMode=RCIM_Cubic, leaveTangent=6.4436, time=0.8699, value=0.7849}
        {interpMode=RCIM_Cubic, tangentMode=RCTM_User, time=0.9119, value=0}
        <at 0.9565 -0.6736 RCIM_Cubic>
        <at 1 0>
       ]}
     }}
   )
  ]
  particleSpawnScript [/Game/Effects/Niagara/Modules/Kite/FN_Initialize_Kite_v2(Module.Write Color=true, Module.Write RibbonTwist=false, Module.Write RibbonWidth=true)]
  particleUpdateScript [
   /Game/Effects/Niagara/Modules/Skydiving/FN_DivingAttributes/FN_DivingAttributes(Accumulate Dive=0.003)
   /Game/Effects/Niagara/Modules/Kite/FN_Update_Kite_v2(
    Amplitude=40
    Axis to match Bone Ref Axis:Vector3f=<0 1 0>
    Bone Reference Axis:Vector3f=<1 0 0>
    Frequency=10
    Kite Scale:Vector3f=&VectorFromFloat(Value=0.8)
    Rotation Amplitude=20
    Scale Kite Position=-31.6522
    Scale String Start Pos=14.4
    Spool Position Offset:Vector3f=<18.9256 -0.1236 2.5576>
    Spool Scale:Vector3f=&VectorFromFloat(Value=0.8)
    String End Offset:Vector3f=<59.6 34.834 97.7>
    String Length Limit=97.79
    String Scale:Vector3f=&VectorFromFloat(Value=0.6)
   )
   /Game/Effects/Niagara/Modules/Materials/FN_MeshMaterial(Amplitude=300, Falloff=416.769, Frequency=200, Scale Noise Y Axis=1.5, Scale Noise Z Axis=0, Speed=1)
   /SetVariables(Particles.FN_KiteMeshAttributes.Orientation:Quat4f=&MultiplyQuaternion(
     Quaternion A=$Particles.FN_KiteMeshAttributes.Orientation
     Quaternion B=&FindQuatBetweenTwoVectors(
      From Vector=&RotateVectorByQuaternion(Vector To Rotate=<0 1 0>, Quaternion=$Particles.FN_KiteMeshAttributes.Orientation)
      To Vector=$Engine.Owner.SystemYAxis
     )
    )) {enabled=false, isSetParameters=true}
  ]
  rendererClasses [&NiagaraMeshRendererProperties, &NiagaraRibbonRendererProperties]
  renderers [
   {rendererClass=&NiagaraMeshRendererProperties, rendererIndex=0}
   {rendererClass=&NiagaraMeshRendererProperties, rendererIndex=1}
   {rendererClass=&NiagaraMeshRendererProperties, rendererIndex=2}
   {rendererClass=&NiagaraRibbonRendererProperties, rendererIndex=3}
  ]
 }
 FinTrail002 {
  emitterUpdateScript [
   &EmitterState(Life Cycle Mode=::NewEnumerator1, Loop Behavior=::NewEnumerator1, Loop Duration Mode=::NewEnumerator1, UseLoopDelay=false)
   &SpawnBurst_Instantaneous(Spawn Count=60, Use Spawn Probability=false, UseLoopCountLimit=false)
  ]
  particleSpawnScript [
   &PartitionParticles(Number Of Partitions=4)
   &InitializeParticle(
    Color=&MaskLinearColor(
     Mask Source=1
     Mask Parameter=$Particles.PartitionParticles.Partition
     Mask=0
     Masked Linear Color=<rgba 0.8438 0.1182 0.0017 1>
     Passthrough Linear Color=<rgba 0 0.0644 0.651 1>
    )
    Ribbon Width=50
    Ribbon Width Mode=::NewEnumerator1
    UsePositionOffset=false
    Write Material Random=true
   )
   &DynamicMaterialParameters(
    Index 0 Linear Color:LinearColor=&SelectLinearColorFromArray(
     Array Sampling Mode=::NewEnumerator1
     Direct Array Index=$Particles.PartitionParticles.Partition
     Color Selection Array=DataInterface() {propertyValues {ColorData [<rgba 0.8073 0.0242 0 1>, <rgba 0.0038 0.0699 0.5312 2>, <rgba 0.8073 0.0242 0 1>, <rgba 0.0038 0.0699 0.5312 2>]}}
     Random Seed=?
    )
    Write Parameter Index 0=::NewEnumerator5
   ) {enabled=false}
   &Color(Color=&SelectLinearColorFromArray(
     Array Sampling Mode=::NewEnumerator1
     Direct Array Index=$Particles.PartitionParticles.Partition
     Color Selection Array=DataInterface() {propertyValues {ColorData [<rgba 2 0.1653 0 1>, <rgba 0.0109 0.1973 1.5 1>, <rgba 2 0.1653 0 1>, <rgba 0.0109 0.1973 1.5 1>]}}
     Random Seed=?
    ))
   &Color() {enabled=false, module=Color001}
  ]
  particleUpdateScript [
   &ParticleState(Kill Particles When Lifetime Has Elapsed=false)
   &SkeletalMeshLocation(
    Bone / Socket Index=$Particles.PartitionParticles.Partition
    MaskSpawnGroup=false
    Mesh Sampling Type=::NewEnumerator5
    Orientation Sampling=::NewEnumerator1
    Sampled Position Offset=&AddVectorToPosition(
     Position=&AddVectorToPosition(
      Position=&SelectVectorFromArray(
       Direct Array Index=$Particles.PartitionParticles.Partition
       Vector Selection Array=DataInterface() {propertyValues {FloatData [<40 40 70>, <-40 40 70>]}}
      )
      Vector=&Lerp_Vector(
       A=&Lerp_Vector(A=<0 0 300>, B=<0 0 400>, Alpha=$User.ParaGlideLeanAlpha)
       B=<0 0 0>
       Alpha=&ReturnNormalizedExecIndex()
      )
     )
     Vector=&Lerp_Vector(
      A=<0 0 0>
      B=<0 0 20>
      Alpha=&Waveform(
       "Waveform [1]"=::NewEnumerator0
       "[1] Phase Scale"=&SelectFloatFromArray(Direct Array Index=$Particles.PartitionParticles.Partition, Float Selection Array=DataInterface() {propertyValues {FloatData [0.5,0.7]}})
       "[1] Frequency"=&SelectFloatFromArray(Direct Array Index=$Particles.PartitionParticles.Partition, Float Selection Array=DataInterface() {propertyValues {FloatData [0.5,0.7]}})
       "[1] Amplitude Min/Max"=<0 1>
       "[1] Scale"=1.5
      )
     )
    )
    Skeletal Mesh=$System.Skeletal Mesh Reader
    Socket Sampling Mode=::NewEnumerator5
   ) {enabled=false}
   &SkeletalMeshLocation(
    Bone / Socket Index=&SelectIntFromArray(Direct Array Index=$Particles.PartitionParticles.Partition, Int Selection Array=DataInterface() {propertyValues {IntData [0,1,0,1]}})
    MaskSpawnGroup=false
    Mesh Sampling Type=::NewEnumerator5
    Orientation Sampling=::NewEnumerator1
    Sampled Position Offset=&AddVectorToPosition(
     Position=&AddVectorToPosition(
      Position=&SelectVectorFromArray(
       Direct Array Index=$Particles.PartitionParticles.Partition
       Vector Selection Array=DataInterface() {propertyValues {FloatData [<0 0 6>, <0 0 0>, <-15 0 20>, <15 0 20>]}}
      )
      Vector=&Lerp_Vector(
       A=&Lerp_Vector(
        A=&SelectVectorFromArray(
         Direct Array Index=$Particles.PartitionParticles.Partition
         Vector Selection Array=DataInterface() {propertyValues {FloatData [<0 0 250>, <0 0 250>, <30 0 200>, <-30 0 200>]}}
        )
        B=&SelectVectorFromArray(
         Direct Array Index=$Particles.PartitionParticles.Partition
         Vector Selection Array=DataInterface() {propertyValues {FloatData [<0 0 350>, <0 0 350>, <30 0 350>, <-30 0 350>]}}
        )
        Alpha=$User.ParaGlideLeanAlpha
       )
       B=<0 0 0>
       Alpha=&ReturnNormalizedExecIndex()
      )
     )
     Vector=&Lerp_Vector(
      A=<0 0 5>
      B=<0 0 10>
      Alpha=&Waveform(
       "Waveform [1]"=::NewEnumerator0
       "[1] Phase Scale"=&SelectFloatFromArray(Direct Array Index=$Particles.PartitionParticles.Partition, Float Selection Array=DataInterface() {propertyValues {FloatData [0.5,0.7]}})
       "[1] Frequency"=&SelectFloatFromArray(Direct Array Index=$Particles.PartitionParticles.Partition, Float Selection Array=DataInterface() {propertyValues {FloatData [0.5,0.7]}})
       "[1] Amplitude Min/Max"=<0 1>
       "[1] Scale"=1.5
      )
     )
    )
    Sampled Position Offset Coordinate Space:ENiagaraSkelMeshTransforms=::NewEnumerator2
    Skeletal Mesh=$System.Skeletal Mesh
    Socket Sampling Mode=::NewEnumerator5
   ) {module=SkeletalMeshLocation001}
   &self:ScratchModule(
    Age:NiagaraFloat=$Emitter.Age
    Curve for Colors=DataInterface() {propertyValues {
      AlphaCurve {keys [<at 0 1>,<at 1 1>]}
      BlueCurve {keys [
        {interpMode=RCIM_Cubic, leaveTangentWeight=0.0518, tangentMode=RCTM_User, tangentWeightMode=RCTWM_WeightedBoth, time=0.0031, value=0}
        {
         arriveTangent=0.0052
         arriveTangentWeight=0.1917
         interpMode=RCIM_Cubic
         leaveTangent=0.0052
         leaveTangentWeight=0.1917
         tangentMode=RCTM_User
         tangentWeightMode=RCTWM_WeightedBoth
         time=0.1371
         value=0
        }
        {
         arriveTangent=0.0293
         arriveTangentWeight=0.0977
         interpMode=RCIM_Cubic
         leaveTangent=0.0293
         leaveTangentWeight=0.0977
         tangentMode=RCTM_User
         tangentWeightMode=RCTWM_WeightedBoth
         time=0.3191
         value=0
        }
        {
         arriveTangentWeight=0.1033
         interpMode=RCIM_Cubic
         leaveTangentWeight=0.1033
         tangentMode=RCTM_User
         tangentWeightMode=RCTWM_WeightedBoth
         time=0.5022
         value=0
        }
        {
         arriveTangentWeight=0.1042
         interpMode=RCIM_Cubic
         leaveTangentWeight=0.1042
         tangentMode=RCTM_User
         tangentWeightMode=RCTWM_WeightedBoth
         time=0.6685
         value=0
        }
        {
         arriveTangentWeight=0.0897
         interpMode=RCIM_Cubic
         leaveTangentWeight=0.0897
         tangentMode=RCTM_User
         tangentWeightMode=RCTWM_WeightedBoth
         time=0.8591
         value=0
        }
        {arriveTangentWeight=0.0561, interpMode=RCIM_Cubic, tangentMode=RCTM_User, tangentWeightMode=RCTWM_WeightedBoth, time=1, value=0}
       ]}
      GreenCurve {keys [
        <at 0 0 RCIM_Cubic>
        {
         arriveTangent=-0.0991
         arriveTangentWeight=0.0819
         interpMode=RCIM_Cubic
         leaveTangent=-0.0991
         leaveTangentWeight=0.0819
         tangentMode=RCTM_User
         tangentWeightMode=RCTWM_WeightedBoth
         time=0.1474
         value=1.3
        }
        {
         arriveTangent=-0.0109
         arriveTangentWeight=0.1035
         interpMode=RCIM_Cubic
         leaveTangent=-0.0109
         leaveTangentWeight=0.1035
         tangentMode=RCTM_User
         tangentWeightMode=RCTWM_WeightedBoth
         time=0.3256
         value=-0.5152
        }
        {
         arriveTangent=-0.0362
         arriveTangentWeight=0.088
         interpMode=RCIM_Cubic
         leaveTangent=-0.0362
         leaveTangentWeight=0.088
         tangentMode=RCTM_User
         tangentWeightMode=RCTWM_WeightedBoth
         time=0.5156
         value=2
        }
        {arriveTangent=-4.1446, interpMode=RCIM_Cubic, leaveTangent=-4.1446, time=0.7015, value=-0.7}
        {
         arriveTangentWeight=0.0831
         interpMode=RCIM_Cubic
         leaveTangentWeight=0.0831
         tangentMode=RCTM_User
         tangentWeightMode=RCTWM_WeightedBoth
         time=0.8534
         value=0.6
        }
        <at 1 0>
       ]}
      RedCurve {keys [
        {arriveTangent=2.3812, interpMode=RCIM_Cubic, leaveTangent=2.3812, tangentMode=RCTM_User, time=0, value=0}
        {
         arriveTangent=0.0019
         arriveTangentWeight=0.0857
         interpMode=RCIM_Cubic
         leaveTangent=0.0019
         leaveTangentWeight=0.0857
         tangentMode=RCTM_User
         tangentWeightMode=RCTWM_WeightedBoth
         time=0.0007
         value=0.1996
        }
        {
         arriveTangent=-0.0076
         arriveTangentWeight=0.0935
         interpMode=RCIM_Cubic
         leaveTangent=-0.0076
         leaveTangentWeight=0.0935
         tangentMode=RCTM_User
         tangentWeightMode=RCTWM_WeightedBoth
         time=0.1528
         value=-0.5
        }
        {
         arriveTangent=0.0038
         arriveTangentWeight=0.1046
         interpMode=RCIM_Cubic
         leaveTangent=0.0038
         leaveTangentWeight=0.1046
         tangentMode=RCTM_User
         tangentWeightMode=RCTWM_WeightedBoth
         time=0.3011
         value=0.6
        }
        {
         arriveTangent=0.1
         arriveTangentWeight=0.133
         interpMode=RCIM_Cubic
         leaveTangent=0.1
         leaveTangentWeight=0.133
         tangentMode=RCTM_User
         tangentWeightMode=RCTWM_WeightedBoth
         time=0.5
         value=-0.7
        }
        {
         arriveTangent=0.0115
         arriveTangentWeight=0.1019
         interpMode=RCIM_Cubic
         leaveTangent=0.0115
         leaveTangentWeight=0.1019
         tangentMode=RCTM_User
         tangentWeightMode=RCTWM_WeightedBoth
         time=0.695
         value=0.8
        }
        {
         arriveTangent=0.011
         arriveTangentWeight=0.1017
         interpMode=RCIM_Cubic
         leaveTangent=0.011
         leaveTangentWeight=0.1017
         tangentMode=RCTM_User
         tangentWeightMode=RCTWM_WeightedBoth
         time=0.8289
         value=-0.4
        }
        {
         arriveTangentWeight=0.0783
         interpMode=RCIM_Cubic
         leaveTangentWeight=0.0783
         tangentMode=RCTM_User
         tangentWeightMode=RCTWM_WeightedBoth
         time=0.9985
         value=0.2005
        }
       ]}
     }}
    Curve for Colors2=DataInterface() {propertyValues {
      AlphaCurve {keys [<at 0 0>,<at 1 0>]}
      BlueCurve {keys [{
         arriveTangentWeight=0.0518
         interpMode=RCIM_Cubic
         leaveTangentWeight=0.0518
         tangentMode=RCTM_User
         tangentWeightMode=RCTWM_WeightedBoth
         time=0.0031
         value=0
        }]}
      GreenCurve {keys [
        {interpMode=RCIM_Cubic, leaveTangentWeight=0.0491, tangentMode=RCTM_User, time=0, value=0}
        {
         arriveTangent=0.0991
         arriveTangentWeight=0.0819
         interpMode=RCIM_Cubic
         leaveTangent=0.0991
         leaveTangentWeight=0.0819
         tangentMode=RCTM_User
         tangentWeightMode=RCTWM_WeightedBoth
         time=0.0936
         value=-0.6202
        }
        {
         arriveTangent=0.0109
         arriveTangentWeight=0.1035
         interpMode=RCIM_Cubic
         leaveTangent=0.0109
         leaveTangentWeight=0.1035
         tangentMode=RCTM_User
         tangentWeightMode=RCTWM_WeightedBoth
         time=0.2719
         value=0.3232
        }
        {
         arriveTangent=0.0362
         arriveTangentWeight=0.088
         interpMode=RCIM_Cubic
         leaveTangent=0.0362
         leaveTangentWeight=0.088
         tangentMode=RCTM_User
         tangentWeightMode=RCTWM_WeightedBoth
         time=0.482
         value=-0.6475
        }
        {
         arriveTangent=-0.0514
         arriveTangentWeight=0.2641
         interpMode=RCIM_Cubic
         leaveTangent=-0.0514
         leaveTangentWeight=0.2159
         tangentMode=RCTM_User
         time=0.6141
         value=0.3351
        }
        {
         arriveTangentWeight=0.0831
         interpMode=RCIM_Cubic
         leaveTangentWeight=0.0831
         tangentMode=RCTM_User
         tangentWeightMode=RCTWM_WeightedBoth
         time=0.8265
         value=-0.4149
        }
        {tangentMode=RCTM_User, time=1, value=0}
       ]}
      RedCurve {keys [
        {
         arriveTangent=0.0196
         arriveTangentWeight=0.0615
         interpMode=RCIM_Cubic
         leaveTangent=0.0196
         leaveTangentWeight=0.0615
         tangentMode=RCTM_User
         tangentWeightMode=RCTWM_WeightedBoth
         time=0
         value=0
        }
        {
         arriveTangent=0.0076
         arriveTangentWeight=0.0935
         interpMode=RCIM_Cubic
         leaveTangent=0.0076
         leaveTangentWeight=0.0935
         tangentMode=RCTM_User
         tangentWeightMode=RCTWM_WeightedBoth
         time=0.1058
         value=0.5
        }
        {
         arriveTangent=-0.0038
         arriveTangentWeight=0.1046
         interpMode=RCIM_Cubic
         leaveTangent=-0.0038
         leaveTangentWeight=0.1046
         tangentMode=RCTM_User
         tangentWeightMode=RCTWM_WeightedBoth
         time=0.258
         value=-0.6
        }
        {
         arriveTangent=-0.1
         arriveTangentWeight=0.133
         interpMode=RCIM_Cubic
         leaveTangent=-0.1
         leaveTangentWeight=0.133
         tangentMode=RCTM_User
         tangentWeightMode=RCTWM_WeightedBoth
         time=0.4126
         value=0.7
        }
        {
         arriveTangent=-0.0115
         arriveTangentWeight=0.1019
         interpMode=RCIM_Cubic
         leaveTangent=-0.0115
         leaveTangentWeight=0.1019
         tangentMode=RCTM_User
         tangentWeightMode=RCTWM_WeightedBoth
         time=0.6144
         value=-0.8
        }
        {
         arriveTangent=-0.011
         arriveTangentWeight=0.1017
         interpMode=RCIM_Cubic
         leaveTangent=-0.011
         leaveTangentWeight=0.1017
         tangentMode=RCTM_User
         tangentWeightMode=RCTWM_WeightedBoth
         time=0.7953
         value=0.4
        }
        {
         arriveTangentWeight=0.1131
         interpMode=RCIM_Cubic
         leaveTangentWeight=0.1131
         tangentMode=RCTM_User
         tangentWeightMode=RCTWM_WeightedBoth
         time=0.9985
         value=0
        }
       ]}
     }}
    Lerp Curve:NiagaraFloat=&FloatFromCurve(
     FloatCurve=DataInterface() {propertyValues {Curve {keys [<at 0 1 RCIM_Constant>, {arriveTangent=-2.5175, interpMode=RCIM_Cubic, leaveTangent=-2.5175, time=0.2066, value=1}, <at 0.3972 0>]}}}
     CurveIndex=&ReturnNormalizedExecIndex()
    )
    NormalizeINdex:NiagaraFloat=&ReturnNormalizedExecIndex(Normalized Index Scale=0.75)
    Offset:NiagaraFloat=&Lerp_Float(
     A=&SelectFloatFromArray(
      Direct Array Index=$Particles.PartitionParticles.Partition
      Float Selection Array=DataInterface() {propertyValues {FloatData [0,0.35,0,0.35]}}
     )
     B=&SelectFloatFromArray(
      Direct Array Index=$Particles.PartitionParticles.Partition
      Float Selection Array=DataInterface() {propertyValues {FloatData [0,0.35,0.25,0.55]}}
     )
     Alpha=&FloatFromCurve(
      FloatCurve=DataInterface() {propertyValues {Curve {keys [<at 0 1 RCIM_Constant>, {arriveTangent=-2.2969, interpMode=RCIM_Cubic, leaveTangent=-2.2969, time=0.2947, value=1}, <at 0.4354 0>]}}}
      CurveIndex=&ReturnNormalizedExecIndex()
     )
    )
    STR:NiagaraFloat=&Multiply_Float(
     A=&Multiply_Float(A=80, B=&Lerp_Float(A=0.6, B=0.4, Alpha=$User.ParaGlideLeanAlpha))
     B=&FloatFromCurve(
      FloatCurve=DataInterface() {propertyValues {Curve {keys [
          {arriveTangent=0.2683, interpMode=RCIM_Cubic, leaveTangent=0.2683, tangentMode=RCTM_User, time=0.0033, value=1}
          {arriveTangent=-1.5254, interpMode=RCIM_Cubic, leaveTangent=-1.5254, tangentMode=RCTM_User, time=0.4983, value=0.2}
          <at 0.8142 0.06>
         ]}}}
      CurveIndex=&ReturnNormalizedExecIndex()
     )
    )
    Speed=0.5
   ) {module=ScratchModule}
   &ScaleRibbonWidth(Ribbon Width Scale=&FloatFromCurve(
     FloatCurve=DataInterface() {propertyValues {Curve {keys [
         {arriveTangent=0.0894, interpMode=RCIM_Cubic, leaveTangent=0.0894, tangentMode=RCTM_User, time=0, value=1}
         {arriveTangent=0.116, interpMode=RCIM_Constant, time=0.4219, value=0.8}
         {arriveTangent=-0.0049, interpMode=RCIM_Cubic, leaveTangent=-0.0049, tangentMode=RCTM_User, time=0.5913, value=0}
        ]}}}
     CurveIndex=&ReturnNormalizedExecIndex()
     Scale Curve=&Multiply_Float(
      A=&Lerp_Float(
       A=&SelectFloatFromArray(Direct Array Index=$Particles.PartitionParticles.Partition, Float Selection Array=DataInterface() {propertyValues {FloatData [1,1,0.6,0.7]}})
       B=0.6
       Alpha=$System.FN_Contrail_Setup.Dive Value
      )
      B=$System.FEsizeMatch
     )
    ))
   &ScaleColor(
    Scale Alpha=&Multiply_Float(
     A=&FloatFromCurve(
      FloatCurve=DataInterface() {propertyValues {Curve {
         defaultValue=1
         keys [
          {arriveTangent=0.2069, interpMode=RCIM_Cubic, leaveTangent=0.2069, tangentMode=RCTM_User, time=0, value=0}
          {arriveTangent=1, interpMode=RCIM_Cubic, leaveTangent=1, time=0.4336, value=0.4242}
          {tangentMode=RCTM_User, time=1, value=1}
         ]
        }}}
      CurveIndex=&ReturnNormalizedExecIndex()
      Scale Curve=&FN_FrontEndFloatModifier(Gameplay=1, Front End=0.5)
     )
     B=&Lerp_Float(A=1.5, B=0.7, Alpha=$System.FN_Contrail_Setup.Dive Value)
    )
    Scale RGB=&VectorFromFloat(Value=&FNTimeOfDayScalar(Night Scale=0.5, Morning Scale=0.8, Evening Scale=0.8))
    ScaleA=true
    ScaleRGB=true
   )
   /SetVariables(Particles.RibbonID=/Niagara/DynamicInputs/NiagaraID/MakeNiagaraIDFromInt(Index:NiagaraInt32=$Particles.PartitionParticles.Partition, Acquire Tag:NiagaraInt32=$Particles.PartitionParticles.Partition)) {isSetParameters=true, module=SetVariables_#3}
   /SetVariables(Particles.ColorAdd:LinearColor=&Multiply_LinearColorByFloat(
     Linear Color=&SelectLinearColorFromArray(
      Array Sampling Mode=::NewEnumerator1
      Direct Array Index=$Particles.PartitionParticles.Partition
      Color Selection Array=DataInterface() {propertyValues {ColorData [<rgba 1 0.0642 0 1>, <rgba 0 0.1611 1 1>, <rgba 1 0.0642 0 1>, <rgba 0 0.1611 1 1>]}}
      Random Seed=?
     )
     Float=&Multiply_Float(
      A=&FloatFromCurve(
       FloatCurve=DataInterface() {propertyValues {Curve {
          defaultValue=1
          keys [
           {arriveTangent=0.2069, interpMode=RCIM_Cubic, leaveTangent=0.2069, tangentMode=RCTM_User, time=0, value=0}
           {arriveTangent=1, interpMode=RCIM_Cubic, leaveTangent=1, time=0.4336, value=0.6818}
           {tangentMode=RCTM_User, time=1, value=1}
          ]
         }}}
       CurveIndex=&ReturnNormalizedExecIndex()
      )
      B=&Lerp_Float(A=0.5, B=0.2, Alpha=$System.FN_Contrail_Setup.Dive Value)
     )
    )) {isSetParameters=true, module=SetVariables_#4}
  ]
  rendererClasses [&NiagaraRibbonRendererProperties]
  renderers [
   {rendererClass=&NiagaraRibbonRendererProperties}
   {rendererClass=&NiagaraRibbonRendererProperties, rendererIndex=1}
  ]
 }
 Gizmo_LocalSpace {
  bEnabled=false
  emitterUpdateScript [
   &EmitterState()
   &SpawnBurst_Instantaneous(Spawn Count=4, Spawn Probability=HlslExpression() {hlslExpression="Emitter.LoopCount == 0"}, Use Spawn Probability=true)
  ]
  particleSpawnScript [
   &InitializeParticle(
    Color=<rgba 0 0.3677 1 1>
    Lifetime Max=1
    Lifetime Mode=::NewEnumerator1
    Mesh Scale Mode=::NewEnumerator3
    Mesh Uniform Scale=3
    UsePositionOffset=false
    Write Material Random=true
   )
   &InitialMeshOrientation(Mesh Axis To Orient=<0 0 -1>, Mesh Orientation Mode=::NewEnumerator2, Use Rotation Vector=false)
  ]
  particleUpdateScript [
   &ParticleState(Kill Particles When Lifetime Has Elapsed=false, Let Infinitely Lived Particles Die When Emitter Deactivates=true)
   &UpdateMeshOrientation(
    Facing Direction=$System.FN_Contrail_Setup.Direction Vector
    Facing Interpolation=false
    Orientation Method=::NewEnumerator1
    Override Quat=false
    Stabilization Mode:ENiagara_OrientVectorCount=::NewEnumerator2
    TransformFacing=false
    TransformSide=false
   )
   /SetVariables(Particles.Position=&AddVectorToPosition(
     Position=$Particles.Initial.Position
     Vector=&Multiply_VectorByFloat(
      Vector=$System.FN_Contrail_Setup.Direction Velocity
      Float=&FloatFromCurve(
       FloatCurve=DataInterface() {propertyValues {Curve {keys [{tangentMode=RCTM_User, time=0, value=0}, {tangentMode=RCTM_User, time=1, value=1}]}}}
       Scale Curve=$Particles.Initial.Lifetime
      )
     )
    )) {enabled=false, isSetParameters=true, module=SetVariables_#5}
   /SetVariables(Particles.Velocity=$System.FN_Contrail_Setup.Direction Vector) {isSetParameters=true, module=SetVariables_#6}
  ]
  rendererClasses [&NiagaraMeshRendererProperties]
  renderers [{rendererClass=&NiagaraMeshRendererProperties, rendererIndex=0}]
 }
 Main001 {
  emitterUpdateScript [
   &EmitterState(Life Cycle Mode=::NewEnumerator1, Loop Behavior=::NewEnumerator1, Loop Duration Mode=::NewEnumerator1, UseLoopDelay=false)
   &SpawnBurst_Instantaneous(Spawn Count=60, Use Spawn Probability=false, UseLoopCountLimit=false)
  ]
  particleSpawnScript [
   &PartitionParticles()
   &InitializeParticle(
    Color=&MaskLinearColor(
     Mask Source=1
     Mask Parameter=$Particles.PartitionParticles.Partition
     Mask=0
     Masked Linear Color=<rgba 0.8438 0.1182 0.0017 1>
     Passthrough Linear Color=<rgba 0 0.0644 0.651 1>
    )
    Ribbon Width=50
    Ribbon Width Mode=::NewEnumerator1
    UsePositionOffset=false
    Write Material Random=true
   )
   &DynamicMaterialParameters(
    Index 0 Linear Color:LinearColor=&MaskLinearColor(
     Mask Source=1
     Mask Parameter=$Particles.PartitionParticles.Partition
     Mask=0
     Masked Linear Color=<rgba 0.8073 0.0242 0 1>
     Passthrough Linear Color=<rgba 0.0038 0.0699 0.5312 2>
    )
    Write Parameter Index 0=::NewEnumerator5
   ) {enabled=false}
   &Color(Color=&MaskLinearColor(
     Mask Source=1
     Mask Parameter=$Particles.PartitionParticles.Partition
     Mask=0
     Masked Linear Color=<rgba 2 0.1653 0 1.5>
     Passthrough Linear Color=<rgba 0.0109 0.1973 1.5 2>
    ))
   &DynamicMaterialParameters(
    Index 0 Param 1=1
    Index 0 Param 4=1
    Module.Index0Float0Write=true
    Module.Index0Float1Write=true
    Module.Index0Float2Write=true
    Module.Index0Float3Write=true
   ) {module=DynamicMaterialParameters001}
   /SetVariables(Particles.AuraDyn:Vector4f=<0 0 0 1>) {isSetParameters=true}
  ]
  particleUpdateScript [
   &ParticleState(Kill Particles When Lifetime Has Elapsed=false)
   &SkeletalMeshLocation(
    Bone / Socket Index=$Particles.PartitionParticles.Partition
    MaskSpawnGroup=false
    Mesh Sampling Type=::NewEnumerator5
    Orientation Sampling=::NewEnumerator1
    Sampled Position Offset=&AddVectorToPosition(
     Position=&AddVectorToPosition(
      Position=&SelectVectorFromArray(
       Direct Array Index=$Particles.PartitionParticles.Partition
       Vector Selection Array=DataInterface() {propertyValues {FloatData [<40 40 70>, <-40 40 70>]}}
      )
      Vector=&Lerp_Vector(
       A=&Lerp_Vector(A=<0 0 300>, B=<0 0 400>, Alpha=$User.ParaGlideLeanAlpha)
       B=<0 0 0>
       Alpha=&ReturnNormalizedExecIndex()
      )
     )
     Vector=&Lerp_Vector(
      A=<0 0 0>
      B=<0 0 20>
      Alpha=&Waveform(
       "Waveform [1]"=::NewEnumerator0
       "[1] Phase Scale"=&SelectFloatFromArray(Direct Array Index=$Particles.PartitionParticles.Partition, Float Selection Array=DataInterface() {propertyValues {FloatData [0.5,0.7]}})
       "[1] Frequency"=&SelectFloatFromArray(Direct Array Index=$Particles.PartitionParticles.Partition, Float Selection Array=DataInterface() {propertyValues {FloatData [0.5,0.7]}})
       "[1] Amplitude Min/Max"=<0 1>
       "[1] Scale"=1.5
      )
     )
    )
    Skeletal Mesh=$System.Skeletal Mesh Reader
    Socket Sampling Mode=::NewEnumerator5
   ) {enabled=false}
   &SkeletalMeshLocation(
    Bone / Socket Index=$Particles.PartitionParticles.Partition
    MaskSpawnGroup=false
    Mesh Sampling Type=::NewEnumerator5
    Orientation Sampling=::NewEnumerator1
    Sampled Position Offset=&AddVectorToPosition(
     Position=&AddVectorToPosition(
      Position=&SelectVectorFromArray(
       Direct Array Index=$Particles.PartitionParticles.Partition
       Vector Selection Array=DataInterface() {propertyValues {FloatData [<0 0 0>, <0 0 0>]}}
      )
      Vector=&Lerp_Vector(
       A=&Lerp_Vector(A=<0 0 250>, B=<0 0 400>, Alpha=$User.ParaGlideLeanAlpha)
       B=<0 0 0>
       Alpha=&ReturnNormalizedExecIndex()
      )
     )
     Vector=&Lerp_Vector(
      A=<0 0 5>
      B=<0 0 10>
      Alpha=&Waveform(
       "Waveform [1]"=::NewEnumerator0
       "[1] Phase Scale"=&SelectFloatFromArray(Direct Array Index=$Particles.PartitionParticles.Partition, Float Selection Array=DataInterface() {propertyValues {FloatData [0.5,0.7]}})
       "[1] Frequency"=&SelectFloatFromArray(Direct Array Index=$Particles.PartitionParticles.Partition, Float Selection Array=DataInterface() {propertyValues {FloatData [0.5,0.7]}})
       "[1] Amplitude Min/Max"=<0 1>
       "[1] Scale"=1.5
      )
     )
    )
    Sampled Position Offset Coordinate Space:ENiagaraSkelMeshTransforms=::NewEnumerator2
    Skeletal Mesh=$System.Skeletal Mesh
    Socket Sampling Mode=::NewEnumerator5
   ) {module=SkeletalMeshLocation001}
   &self:ScratchModule(
    Age:NiagaraFloat=$Emitter.Age
    Curve for Colors=DataInterface() {propertyValues {
      AlphaCurve {keys [<at 0 1>,<at 1 1>]}
      BlueCurve {keys [
        {interpMode=RCIM_Cubic, leaveTangentWeight=0.0518, tangentMode=RCTM_User, tangentWeightMode=RCTWM_WeightedBoth, time=0.0031, value=0}
        {
         arriveTangent=0.0052
         arriveTangentWeight=0.1917
         interpMode=RCIM_Cubic
         leaveTangent=0.0052
         leaveTangentWeight=0.1917
         tangentMode=RCTM_User
         tangentWeightMode=RCTWM_WeightedBoth
         time=0.1371
         value=0
        }
        {
         arriveTangent=0.0293
         arriveTangentWeight=0.0977
         interpMode=RCIM_Cubic
         leaveTangent=0.0293
         leaveTangentWeight=0.0977
         tangentMode=RCTM_User
         tangentWeightMode=RCTWM_WeightedBoth
         time=0.3191
         value=0
        }
        {
         arriveTangentWeight=0.1033
         interpMode=RCIM_Cubic
         leaveTangentWeight=0.1033
         tangentMode=RCTM_User
         tangentWeightMode=RCTWM_WeightedBoth
         time=0.5022
         value=0
        }
        {
         arriveTangentWeight=0.1042
         interpMode=RCIM_Cubic
         leaveTangentWeight=0.1042
         tangentMode=RCTM_User
         tangentWeightMode=RCTWM_WeightedBoth
         time=0.6685
         value=0
        }
        {
         arriveTangentWeight=0.0897
         interpMode=RCIM_Cubic
         leaveTangentWeight=0.0897
         tangentMode=RCTM_User
         tangentWeightMode=RCTWM_WeightedBoth
         time=0.8591
         value=0
        }
        {arriveTangentWeight=0.0561, interpMode=RCIM_Cubic, tangentMode=RCTM_User, tangentWeightMode=RCTWM_WeightedBoth, time=1, value=0}
       ]}
      GreenCurve {keys [
        <at 0 0 RCIM_Cubic>
        {
         arriveTangent=-0.0991
         arriveTangentWeight=0.0819
         interpMode=RCIM_Cubic
         leaveTangent=-0.0991
         leaveTangentWeight=0.0819
         tangentMode=RCTM_User
         tangentWeightMode=RCTWM_WeightedBoth
         time=0.1474
         value=1.3
        }
        {
         arriveTangent=-0.0109
         arriveTangentWeight=0.1035
         interpMode=RCIM_Cubic
         leaveTangent=-0.0109
         leaveTangentWeight=0.1035
         tangentMode=RCTM_User
         tangentWeightMode=RCTWM_WeightedBoth
         time=0.3256
         value=-0.5152
        }
        {
         arriveTangent=-0.0362
         arriveTangentWeight=0.088
         interpMode=RCIM_Cubic
         leaveTangent=-0.0362
         leaveTangentWeight=0.088
         tangentMode=RCTM_User
         tangentWeightMode=RCTWM_WeightedBoth
         time=0.5156
         value=2
        }
        {arriveTangent=-4.1446, interpMode=RCIM_Cubic, leaveTangent=-4.1446, time=0.7015, value=-0.7}
        {
         arriveTangentWeight=0.0831
         interpMode=RCIM_Cubic
         leaveTangentWeight=0.0831
         tangentMode=RCTM_User
         tangentWeightMode=RCTWM_WeightedBoth
         time=0.8534
         value=0.6
        }
        <at 1 0>
       ]}
      RedCurve {keys [
        {arriveTangent=2.3812, interpMode=RCIM_Cubic, leaveTangent=2.3812, tangentMode=RCTM_User, time=0, value=0}
        {
         arriveTangent=0.0019
         arriveTangentWeight=0.0857
         interpMode=RCIM_Cubic
         leaveTangent=0.0019
         leaveTangentWeight=0.0857
         tangentMode=RCTM_User
         tangentWeightMode=RCTWM_WeightedBoth
         time=0.0007
         value=0.1996
        }
        {
         arriveTangent=-0.0076
         arriveTangentWeight=0.0935
         interpMode=RCIM_Cubic
         leaveTangent=-0.0076
         leaveTangentWeight=0.0935
         tangentMode=RCTM_User
         tangentWeightMode=RCTWM_WeightedBoth
         time=0.1528
         value=-0.5
        }
        {
         arriveTangent=0.0038
         arriveTangentWeight=0.1046
         interpMode=RCIM_Cubic
         leaveTangent=0.0038
         leaveTangentWeight=0.1046
         tangentMode=RCTM_User
         tangentWeightMode=RCTWM_WeightedBoth
         time=0.3011
         value=0.6
        }
        {
         arriveTangent=0.1
         arriveTangentWeight=0.133
         interpMode=RCIM_Cubic
         leaveTangent=0.1
         leaveTangentWeight=0.133
         tangentMode=RCTM_User
         tangentWeightMode=RCTWM_WeightedBoth
         time=0.5
         value=-0.7
        }
        {
         arriveTangent=0.0115
         arriveTangentWeight=0.1019
         interpMode=RCIM_Cubic
         leaveTangent=0.0115
         leaveTangentWeight=0.1019
         tangentMode=RCTM_User
         tangentWeightMode=RCTWM_WeightedBoth
         time=0.695
         value=0.8
        }
        {
         arriveTangent=0.011
         arriveTangentWeight=0.1017
         interpMode=RCIM_Cubic
         leaveTangent=0.011
         leaveTangentWeight=0.1017
         tangentMode=RCTM_User
         tangentWeightMode=RCTWM_WeightedBoth
         time=0.8289
         value=-0.4
        }
        {
         arriveTangentWeight=0.0783
         interpMode=RCIM_Cubic
         leaveTangentWeight=0.0783
         tangentMode=RCTM_User
         tangentWeightMode=RCTWM_WeightedBoth
         time=0.9985
         value=0.2005
        }
       ]}
     }}
    NormalizeINdex:NiagaraFloat=&ReturnNormalizedExecIndex(Normalized Index Scale=0.75)
    Offset:NiagaraFloat=&SelectFloatFromArray(Direct Array Index=$Particles.PartitionParticles.Partition, Float Selection Array=DataInterface() {propertyValues {FloatData [0,0.35]}})
    STR:NiagaraFloat=&Multiply_Float(
     A=&Multiply_Float(A=70, B=&Lerp_Float(A=0.6, B=0.8, Alpha=$User.ParaGlideLeanAlpha))
     B=&FloatFromCurve(
      FloatCurve=DataInterface() {propertyValues {Curve {keys [
          {arriveTangent=-0.0156, interpMode=RCIM_Cubic, leaveTangent=-0.0156, tangentMode=RCTM_User, time=0, value=1}
          {arriveTangent=-1.5254, interpMode=RCIM_Cubic, leaveTangent=-1.5254, tangentMode=RCTM_User, time=0.4983, value=0.25}
          <at 0.8142 0.08>
         ]}}}
      CurveIndex=&ReturnNormalizedExecIndex()
     )
    )
    Speed=0.5
   ) {module=ScratchModule}
   &ScaleRibbonWidth(Ribbon Width Scale=&FloatFromCurve(
     FloatCurve=DataInterface() {propertyValues {Curve {keys [
         {arriveTangent=1.0648, interpMode=RCIM_Cubic, leaveTangent=1.0648, tangentMode=RCTM_User, time=0, value=1.5}
         {arriveTangent=-0.6, interpMode=RCIM_Cubic, leaveTangent=-0.6, time=0.5412, value=0.7794}
         <at 1 0.9>
        ]}}}
     CurveIndex=&ReturnNormalizedExecIndex()
     Scale Curve=&Multiply_Float(A=&Lerp_Float(A=1.2, B=0.9, Alpha=$System.FN_Contrail_Setup.Dive Value), B=$System.FEsizeMatch)
    ))
   &ScaleColor(
    Scale Alpha=&Multiply_Float(
     A=&FloatFromCurve(
      FloatCurve=DataInterface() {propertyValues {Curve {
         defaultValue=1
         keys [
          {arriveTangent=0.2069, interpMode=RCIM_Cubic, leaveTangent=0.2069, tangentMode=RCTM_User, time=0, value=0}
          {arriveTangent=1, interpMode=RCIM_Cubic, leaveTangent=1, time=0.4336, value=0.363}
          {tangentMode=RCTM_User, time=1, value=1}
         ]
        }}}
      CurveIndex=&ReturnNormalizedExecIndex()
      Scale Curve=&FN_FrontEndFloatModifier(Gameplay=1, Front End=0.65)
     )
     B=&Lerp_Float(
      A=1.5
      B=&FN_FrontEndFloatModifier(Gameplay=0.7, Front End=0.4)
      Alpha=$System.FN_Contrail_Setup.Dive Value
     )
    )
    Scale RGB=&VectorFromFloat(Value=&FNTimeOfDayScalar(Night Scale=0.5, Morning Scale=0.8, Evening Scale=0.8))
    ScaleA=true
    ScaleRGB=true
   )
   &ScaleColor(
    Scale Alpha=&Multiply_Float(
     A=&FloatFromCurve(
      FloatCurve=DataInterface() {propertyValues {Curve {
         defaultValue=1
         keys [
          {arriveTangent=0.2069, interpMode=RCIM_Cubic, leaveTangent=0.2069, tangentMode=RCTM_User, time=0, value=0}
          {arriveTangent=1, interpMode=RCIM_Cubic, leaveTangent=1, time=0.4336, value=0.2627}
          {tangentMode=RCTM_User, time=1, value=1}
         ]
        }}}
      CurveIndex=&ReturnNormalizedExecIndex()
     )
     B=&Lerp_Float(A=1.5, B=0.5, Alpha=$System.FN_Contrail_Setup.Dive Value)
    )
    Scale RGB=&VectorFromFloat(Value=&FNTimeOfDayScalar(Night Scale=0.5, Morning Scale=0.8, Evening Scale=0.8))
    ScaleA=true
    ScaleRGB=true
   ) {enabled=false, module=ScaleColor001}
   /SetVariables(Particles.RibbonID=&MakeNiagaraIDFromInt(Index:NiagaraInt32=$Particles.PartitionParticles.Partition, Acquire Tag:NiagaraInt32=$Particles.PartitionParticles.Partition)) {isSetParameters=true, module=SetVariables_#3}
   /SetVariables(Particles.ColorAdd:LinearColor=&Multiply_LinearColorByFloat(
     Linear Color=&MaskLinearColor(
      Mask Source=1
      Mask Parameter=$Particles.PartitionParticles.Partition
      Mask=0
      Masked Linear Color=<rgba 1 0.0642 0 1>
      Passthrough Linear Color=<rgba 0 0.1611 1 1>
     )
     Float=&Multiply_Float(
      A=&FloatFromCurve(
       FloatCurve=DataInterface() {propertyValues {Curve {
          defaultValue=1
          keys [
           {arriveTangent=0.2069, interpMode=RCIM_Cubic, leaveTangent=0.2069, tangentMode=RCTM_User, time=0, value=0}
           {arriveTangent=1, interpMode=RCIM_Cubic, leaveTangent=1, time=0.3618, value=0.6818}
           {tangentMode=RCTM_User, time=1, value=1}
          ]
         }}}
       CurveIndex=&ReturnNormalizedExecIndex()
      )
      B=&Lerp_Float(A=0.5, B=0.3, Alpha=$System.FN_Contrail_Setup.Dive Value)
     )
    )) {isSetParameters=true, module=SetVariables_#4}
   /SetVariables(
    Particles.AuraColor:LinearColor=&Multiply_LinearColorByFloat(
     Linear Color=&MaskLinearColor(
      Mask Source=1
      Mask Parameter=$Particles.PartitionParticles.Partition
      Mask=0
      Masked Linear Color=<rgba 1 0.2583 0 1>
      Passthrough Linear Color=<rgba 0.0124 0.2273 1 1>
     )
     Float=&OneMinusFloat(Float=$System.FN_Contrail_Setup.Dive Value)
    )
    Particles.AuraRibbonWidth:NiagaraFloat=&Multiply_Float(A=$Particles.RibbonWidth, B=0.8)
   ) {isSetParameters=true}
  ]
  rendererClasses [&NiagaraRibbonRendererProperties]
  renderers [
   {rendererClass=&NiagaraRibbonRendererProperties}
   {rendererClass=&NiagaraRibbonRendererProperties, rendererIndex=1}
   {rendererClass=&NiagaraRibbonRendererProperties, rendererIndex=2}
   {rendererClass=&NiagaraRibbonRendererProperties, rendererIndex=3}
  ]
 }
 Mesh_SkeletalMesh {
  bEnabled=false
  emitterUpdateScript [
   &EmitterState()
   &SpawnBurst_Instantaneous(Spawn Count=4, Spawn Probability=HlslExpression() {hlslExpression="Emitter.LoopCount == 0"}, Use Spawn Probability=true)
  ]
  particleSpawnScript [
   &PartitionParticles(Number Of Partitions=4)
   &InitializeParticle(Color=<rgba 0 1 0.0484 1>, Sprite Size=<10 100>, Sprite Size Mode=::NewEnumerator4, UsePositionOffset=false, Write Material Random=true)
   &Color(Color=&SelectLinearColorFromArray(
     Array Sampling Mode=::NewEnumerator1
     Direct Array Index=$Particles.PartitionParticles.Partition
     Color Selection Array=DataInterface() {propertyValues {ColorData [<rgba 1 0 0.139 4>, <rgba 0 0.1466 1 4>]}}
     Random Seed=?
    ))
  ]
  particleUpdateScript [
   &SolveForcesAndVelocity(Clamp Velocity=false, Limit Acceleration=false)
   &ParticleState(Kill Particles When Lifetime Has Elapsed=false, Let Infinitely Lived Particles Die When Emitter Deactivates=true)
   &SkeletalMeshLocation(
    Bone / Socket Index=$Particles.PartitionParticles.Partition
    MaskSpawnGroup=false
    Mesh Sampling Type=::NewEnumerator5
    Orientation Sampling=::NewEnumerator1
    Skeletal Mesh=$System.Skeletal Mesh
    Socket Sampling Mode=::NewEnumerator5
   )
   &UpdateMeshOrientation(
    Facing Direction=$System.FN_Contrail_Setup.Direction Vector
    Facing Interpolation=false
    Orientation Method=::NewEnumerator1
    Override Quat=false
    Stabilization Mode:ENiagara_OrientVectorCount=::NewEnumerator2
    TransformFacing=false
    TransformSide=false
   )
   &AlignSpriteToMeshOrientation(Mesh Orientation Relative Sprite Alignment Vector=<-1 0 0>, SetAlignment=true, SetFacing=false)
  ]
  rendererClasses [&NiagaraMeshRendererProperties, &NiagaraSpriteRendererProperties]
  renderers [
   {rendererClass=&NiagaraMeshRendererProperties, rendererIndex=0}
   {rendererClass=&NiagaraSpriteRendererProperties, rendererIndex=1}
   {rendererClass=&NiagaraMeshRendererProperties, rendererIndex=2}
  ]
 }
 Particle_Root {
  bEnabled=false
  emitterUpdateScript [
   &EmitterState()
   &SpawnRate(SpawnRate=31, Use Spawn Probability=false)
  ]
  particleSpawnScript [
   &InitializeParticle(
    Color=<rgba 0 0.3677 1 1>
    Lifetime Max=1
    Lifetime Mode=::NewEnumerator1
    Sprite Size Mode=::NewEnumerator1
    UsePositionOffset=false
    Write Material Random=true
   )
   /SetVariables(Particles.Velocity=$System.FN_Contrail_Setup.Direction Vector) {enabled=false, isSetParameters=true}
   &ShapeLocation(
    Cylinder Height Midpoint=1
    Cylinder Radius=400
    MaskSpawnGroup=false
    Offset Mode=::NewEnumerator0
    Override Seed=false
    Rotation Axis=$System.FN_Contrail_Setup.Direction Vector
    Rotation Mode=::NewEnumerator1
    Shape Primitive=::NewEnumerator1
    Surface Only=true
    Surface Only Band Thickness=0.75
    Transform Order:ENiagara_TransformOrder=::NewEnumerator1
    Use Endcaps In Surface Only Mode=false
    UseOwnerScale=true
   )
  ]
  particleUpdateScript [
   &ParticleState()
   /SetVariables(Particles.NewAge:NiagaraFloat=&Multiply_Float(A=$Particles.Age, B=1)) {isSetParameters=true}
   /SetVariables(Particles.Velocity=&Multiply_VectorByFloat(
     Vector=&Lerp_Vector(A=$System.FN_Contrail_Setup.Direction Vector, B=$System.FN_Contrail_Setup.Direction Velocity, Alpha=$Particles.NewAge)
     Float=1
    )) {isSetParameters=true}
   /SetVariables(Particles.Position=&AddVectorToPosition(
     Position=$Particles.Initial.Position
     Vector=&Multiply_VectorByFloat(Vector=$Particles.Velocity, Float=$Particles.Age)
    )) {isSetParameters=true, module=SetVariables_#5}
   &ScaleSpriteSize(
    Scale Factor=&Lerp_Vector2(A=<1 4>, B=<1 12>, Alpha=$System.FN_Contrail_Setup.Dive Value)
    Scale Sprite Size Mode=::NewEnumerator2
   )
   &ScaleColor(
    Scale Alpha=&FloatFromCurve(
     FloatCurve=DataInterface() {propertyValues {Curve {keys [
         {interpMode=RCIM_Cubic, tangentMode=RCTM_User, time=0, value=0}
         {interpMode=RCIM_Cubic, tangentMode=RCTM_User, time=0.5, value=1}
         {tangentMode=RCTM_User, time=1, value=0}
        ]}}}
     Scale Curve=$System.FN_Contrail_Setup.Normalized Intro Age
    )
    ScaleA=true
    ScaleRGB=true
   )
  ]
  rendererClasses [&NiagaraSpriteRendererProperties, &NiagaraMeshRendererProperties]
  renderers [
   {rendererClass=&NiagaraSpriteRendererProperties}
   {rendererClass=&NiagaraMeshRendererProperties, rendererIndex=1}
  ]
 }
 Particles_SkeletalMesh {
  bEnabled=false
  emitterUpdateScript [
   &EmitterState()
   &SpawnRate(SpawnRate=20, Use Spawn Probability=false)
  ]
  particleSpawnScript [
   &PartitionParticles()
   &InitializeParticle(
    Color=&MaskLinearColor(
     Mask Source=1
     Mask Parameter=$Particles.PartitionParticles.Partition
     Mask=0
     Masked Linear Color=<rgba 1 0 0.029 1>
     Passthrough Linear Color=<rgba 0 0.099 1 1>
    )
    Lifetime Max=1
    Lifetime Mode=::NewEnumerator1
    Sprite Size=<5 20>
    Sprite Size Mode=::NewEnumerator4
    UsePositionOffset=false
    Write Material Random=true
   )
   &ShapeLocation(
    Apply To Particle Position=::NewEnumerator1
    MaskSpawnGroup=false
    Override Seed=false
    Shape Origin=&VectorFromFloat(Value=0)
    Sphere Radius=5
    UseOwnerScale=true
   )
  ]
  particleUpdateScript [
   &ParticleState()
   &SkeletalMeshLocation(
    Bone / Socket Index=$Particles.PartitionParticles.Partition
    MaskSpawnGroup=false
    Mesh Sampling Type=::NewEnumerator5
    Orientation Sampling=::NewEnumerator1
    Sampled Position Offset=$Particles.ShapeLocation.ShapePosition
    Skeletal Mesh=$System.Skeletal Mesh
    Socket Sampling Mode=::NewEnumerator5
   )
   /SetVariables(Particles.NewAge:NiagaraFloat=&Multiply_Float(A=$Particles.Age, B=0.75)) {isSetParameters=true}
   /SetVariables(Particles.Velocity=&Multiply_VectorByFloat(Vector=&Lerp_Vector(A=<0 0 0>, B=$System.FN_Contrail_Setup.Direction Velocity, Alpha=$Particles.NewAge), Float=1)) {isSetParameters=true}
   /SetVariables(Particles.Position=&AddVectorToPosition(
     Position=$Particles.Position
     Vector=&Multiply_VectorByFloat(Vector=$Particles.Velocity, Float=$Particles.Age)
    )) {isSetParameters=true, module=SetVariables_#5}
   /SetVariables(Particles.Velocity=$System.FN_Contrail_Setup.Direction Vector) {enabled=false, isSetParameters=true, module=SetVariables_#6}
   &ScaleColor(
    Scale Alpha=&FloatFromCurve(
     FloatCurve=DataInterface() {propertyValues {Curve {keys [
         {interpMode=RCIM_Cubic, tangentMode=RCTM_User, time=0, value=0}
         {interpMode=RCIM_Cubic, tangentMode=RCTM_User, time=0.5, value=1}
         {tangentMode=RCTM_User, time=1, value=0}
        ]}}}
     Scale Curve=$System.FN_Contrail_Setup.Normalized Intro Age
    )
    ScaleA=true
    ScaleRGB=true
   )
  ]
  rendererClasses [&NiagaraSpriteRendererProperties]
  renderers [{rendererClass=&NiagaraSpriteRendererProperties}]
 }
 ParticletoRibbon_SkeletalMesh {
  bEnabled=false
  emitterUpdateScript [
   &EmitterState()
   &SpawnBurst_Instantaneous(Spawn Count=40, Use Spawn Probability=false, UseLoopCountLimit=true)
  ]
  particleSpawnScript [
   &PartitionParticles(
    Normalized Index In Partition Includes One=true
    Particles Per Partition=20
    Partition Mode:ENiagara_PartitionMode=::NewEnumerator1
    Write Partition Index to Ribbon Link Order:ENiagara_AttributeSamplingApplyOutput=::NewEnumerator0
    Write Partition to Ribbon ID=::NewEnumerator0
   )
   &InitializeParticle(
    Color=&MaskLinearColor(
     Mask Source=1
     Mask Parameter=$Particles.PartitionParticles.Partition
     Mask=0
     Masked Linear Color=<rgba 1 0.14 0.0021 1>
     Passthrough Linear Color=<rgba 0 0.099 1 1>
    )
    Lifetime=0.5
    Ribbon Width=50
    Ribbon Width Mode=::NewEnumerator1
    Sprite Size Mode=::NewEnumerator3
    UsePositionOffset=false
    Write Material Random=true
   )
  ]
  particleUpdateScript [
   &ParticleState(Kill Particles When Lifetime Has Elapsed=false)
   &SkeletalMeshLocation(
    Bone / Socket Index=$Particles.PartitionParticles.Partition
    MaskSpawnGroup=false
    Mesh Sampling Type=::NewEnumerator5
    Skeletal Mesh=$System.Skeletal Mesh
    Socket Sampling Mode=::NewEnumerator5
   )
   /SetVariables(Particles.NewAge:NiagaraFloat=&Multiply_Float(A=$Particles.PartitionParticles.Normalized Index In Partition, B=$Particles.Initial.Lifetime)) {isSetParameters=true}
   /SetVariables(Particles.Velocity=&Multiply_VectorByFloat(
     Vector=&Lerp_Vector(A=$System.FN_Contrail_Setup.Direction Velocity Raw, B=$System.FN_Contrail_Setup.Direction Velocity, Alpha=$Particles.NewAge)
     Float=$Particles.Initial.Lifetime
    )) {isSetParameters=true}
   /SetVariables(Particles.Position=&AddVectorToPosition(
     Position=$Particles.Position
     Vector=&Multiply_VectorByFloat(Vector=$Particles.Velocity, Float=$Particles.PartitionParticles.Normalized Index In Partition)
    )) {isSetParameters=true, module=SetVariables_#5}
   &ScaleColor(Scale Alpha=$System.FN_Contrail_Setup.Normalized Intro Age, ScaleA=true, ScaleRGB=true)
   &ScaleRibbonWidth(Ribbon Width Scale=&FloatFromCurve(
     FloatCurve=DataInterface() {propertyValues {Curve {
        defaultValue=1
        keys [
         {arriveTangent=5.1933, interpMode=RCIM_Cubic, leaveTangent=5.1933, tangentMode=RCTM_User, time=0, value=0}
         {interpMode=RCIM_Cubic, tangentMode=RCTM_User, time=0.1626, value=1}
         {tangentMode=RCTM_User, time=1, value=0}
        ]
       }}}
     CurveIndex=$Particles.PartitionParticles.Normalized Index In Partition
    ))
  ]
  rendererClasses [&NiagaraSpriteRendererProperties, &NiagaraRibbonRendererProperties]
  renderers [
   {rendererClass=&NiagaraSpriteRendererProperties}
   {rendererClass=&NiagaraRibbonRendererProperties, rendererIndex=1}
  ]
 }
 SideFins001 {
  emitterUpdateScript [
   &EmitterState(Life Cycle Mode=::NewEnumerator1, Loop Behavior=::NewEnumerator1, Loop Duration Mode=::NewEnumerator1, UseLoopDelay=false)
   &SpawnBurst_Instantaneous(Spawn Count=60, Use Spawn Probability=false, UseLoopCountLimit=false)
  ]
  particleSpawnScript [
   &PartitionParticles(Number Of Partitions=8)
   &InitializeParticle(
    Color=&MaskLinearColor(
     Mask Source=1
     Mask Parameter=$Particles.PartitionParticles.Partition
     Mask=0
     Masked Linear Color=<rgba 0.8438 0.1182 0.0017 1>
     Passthrough Linear Color=<rgba 0 0.0644 0.651 1>
    )
    Ribbon Width=50
    Ribbon Width Mode=::NewEnumerator1
    UsePositionOffset=false
    Write Material Random=true
   )
   &DynamicMaterialParameters(
    Index 0 Linear Color:LinearColor=&SelectLinearColorFromArray(
     Array Sampling Mode=::NewEnumerator1
     Direct Array Index=$Particles.PartitionParticles.Partition
     Color Selection Array=DataInterface() {propertyValues {ColorData [
        <rgba 0.8073 0.0242 0 1>
        <rgba 0.8073 0.0242 0 1>
        <rgba 0.8073 0.0242 0 1>
        <rgba 0.8073 0.0242 0 1>
        <rgba 0.0027 0.0486 0.3698 2>
        <rgba 0.0027 0.0486 0.3698 2>
        <rgba 0.0027 0.0486 0.3698 2>
        <rgba 0.0027 0.0486 0.3698 2>
       ]}}
     Random Seed=?
    )
    Write Parameter Index 0=::NewEnumerator5
   ) {enabled=false}
   &Color(Color=&SelectLinearColorFromArray(
     Array Sampling Mode=::NewEnumerator1
     Direct Array Index=$Particles.PartitionParticles.Partition
     Color Selection Array=DataInterface() {propertyValues {ColorData [
        <rgba 2 0.1653 0 1>
        <rgba 2 0.1653 0 1>
        <rgba 2 0.1653 0 1>
        <rgba 2 0.1653 0 1>
        <rgba 0.0145 0.263 2 2>
        <rgba 0.0145 0.263 2 2>
        <rgba 0.0145 0.263 2 2>
        <rgba 0.0145 0.263 2 2>
       ]}}
     Random Seed=?
    )) {module=Color001}
  ]
  particleUpdateScript [
   &ParticleState(Kill Particles When Lifetime Has Elapsed=false)
   &SkeletalMeshLocation(
    Bone / Socket Index=$Particles.PartitionParticles.Partition
    MaskSpawnGroup=false
    Mesh Sampling Type=::NewEnumerator5
    Orientation Sampling=::NewEnumerator1
    Sampled Position Offset=&AddVectorToPosition(
     Position=&AddVectorToPosition(
      Position=&SelectVectorFromArray(
       Direct Array Index=$Particles.PartitionParticles.Partition
       Vector Selection Array=DataInterface() {propertyValues {FloatData [<40 40 70>, <-40 40 70>]}}
      )
      Vector=&Lerp_Vector(
       A=&Lerp_Vector(A=<0 0 300>, B=<0 0 400>, Alpha=$User.ParaGlideLeanAlpha)
       B=<0 0 0>
       Alpha=&ReturnNormalizedExecIndex()
      )
     )
     Vector=&Lerp_Vector(
      A=<0 0 0>
      B=<0 0 20>
      Alpha=&Waveform(
       "Waveform [1]"=::NewEnumerator0
       "[1] Phase Scale"=&SelectFloatFromArray(Direct Array Index=$Particles.PartitionParticles.Partition, Float Selection Array=DataInterface() {propertyValues {FloatData [0.5,0.7]}})
       "[1] Frequency"=&SelectFloatFromArray(Direct Array Index=$Particles.PartitionParticles.Partition, Float Selection Array=DataInterface() {propertyValues {FloatData [0.5,0.7]}})
       "[1] Amplitude Min/Max"=<0 1>
       "[1] Scale"=1.5
      )
     )
    )
    Skeletal Mesh=$System.Skeletal Mesh Reader
    Socket Sampling Mode=::NewEnumerator5
   ) {enabled=false}
   &SkeletalMeshLocation(
    Bone / Socket Index=&SelectIntFromArray(Direct Array Index=$Particles.PartitionParticles.Partition, Int Selection Array=DataInterface() {propertyValues {IntData [0,0,0,0,1,1,1,1]}})
    MaskSpawnGroup=false
    Mesh Sampling Type=::NewEnumerator5
    Orientation Sampling=::NewEnumerator1
    Sampled Position Offset=&AddVectorToPosition(
     Position=&AddVectorToPosition(
      Position=&SelectVectorFromArray(
       Direct Array Index=$Particles.PartitionParticles.Partition
       Vector Selection Array=DataInterface() {propertyValues {FloatData [<-16 0 -5>, <16 0 0>, <-16 0 -5>, <16 0 0>, <-16 0 -5>, <15 0 0>, <-16 0 -5>, <15 0 0>]}}
      )
      Vector=&Lerp_Vector(
       A=&Lerp_Vector(
        A=&SelectVectorFromArray(
         Direct Array Index=$Particles.PartitionParticles.Partition
         Vector Selection Array=DataInterface() {propertyValues {FloatData [<55 0 125>, <-60 0 125>, <55 0 125>, <-60 0 125>, <60 0 125>, <-60 0 125>, <60 0 125>, <-60 0 125>]}}
        )
        B=&SelectVectorFromArray(
         Direct Array Index=$Particles.PartitionParticles.Partition
         Vector Selection Array=DataInterface() {propertyValues {FloatData [<55 0 200>, <-60 0 200>, <55 0 200>, <-60 0 200>, <60 0 200>, <-60 0 200>, <60 0 200>, <-60 0 200>]}}
        )
        Alpha=$User.ParaGlideLeanAlpha
       )
       B=<0 0 0>
       Alpha=&ReturnNormalizedExecIndex()
      )
     )
     Vector=&Lerp_Vector(
      A=<0 0 5>
      B=<0 0 10>
      Alpha=&Waveform(
       "Waveform [1]"=::NewEnumerator0
       "[1] Phase Scale"=&SelectFloatFromArray(Direct Array Index=$Particles.PartitionParticles.Partition, Float Selection Array=DataInterface() {propertyValues {FloatData [0.5,0.7]}})
       "[1] Frequency"=&SelectFloatFromArray(Direct Array Index=$Particles.PartitionParticles.Partition, Float Selection Array=DataInterface() {propertyValues {FloatData [0.5,0.7]}})
       "[1] Amplitude Min/Max"=<0 1>
       "[1] Scale"=1.5
      )
     )
    )
    Sampled Position Offset Coordinate Space:ENiagaraSkelMeshTransforms=::NewEnumerator2
    Skeletal Mesh=$System.Skeletal Mesh
    Socket Sampling Mode=::NewEnumerator5
   ) {module=SkeletalMeshLocation001}
   &self:ScratchModule(
    Age:NiagaraFloat=$Emitter.Age
    Curve for Colors=DataInterface() {propertyValues {
      AlphaCurve {keys [<at 0 0>,<at 1 0>]}
      BlueCurve {keys [
        {interpMode=RCIM_Cubic, leaveTangentWeight=0.0518, tangentMode=RCTM_User, tangentWeightMode=RCTWM_WeightedBoth, time=0.0031, value=0}
        {
         arriveTangent=0.0052
         arriveTangentWeight=0.1917
         interpMode=RCIM_Cubic
         leaveTangent=0.0052
         leaveTangentWeight=0.1917
         tangentMode=RCTM_User
         tangentWeightMode=RCTWM_WeightedBoth
         time=0.1371
         value=0
        }
        {
         arriveTangent=0.0293
         arriveTangentWeight=0.0977
         interpMode=RCIM_Cubic
         leaveTangent=0.0293
         leaveTangentWeight=0.0977
         tangentMode=RCTM_User
         tangentWeightMode=RCTWM_WeightedBoth
         time=0.3191
         value=0
        }
        {
         arriveTangentWeight=0.1033
         interpMode=RCIM_Cubic
         leaveTangentWeight=0.1033
         tangentMode=RCTM_User
         tangentWeightMode=RCTWM_WeightedBoth
         time=0.5022
         value=0
        }
        {
         arriveTangentWeight=0.1042
         interpMode=RCIM_Cubic
         leaveTangentWeight=0.1042
         tangentMode=RCTM_User
         tangentWeightMode=RCTWM_WeightedBoth
         time=0.6685
         value=0
        }
        {
         arriveTangentWeight=0.0897
         interpMode=RCIM_Cubic
         leaveTangentWeight=0.0897
         tangentMode=RCTM_User
         tangentWeightMode=RCTWM_WeightedBoth
         time=0.8591
         value=0
        }
        {arriveTangentWeight=0.0561, interpMode=RCIM_Cubic, tangentMode=RCTM_User, tangentWeightMode=RCTWM_WeightedBoth, time=1, value=0}
       ]}
      GreenCurve {keys [
        <at 0 0 RCIM_Cubic>
        {
         arriveTangent=-0.0991
         arriveTangentWeight=0.0819
         interpMode=RCIM_Cubic
         leaveTangent=-0.0991
         leaveTangentWeight=0.0819
         tangentMode=RCTM_User
         tangentWeightMode=RCTWM_WeightedBoth
         time=0.1474
         value=0
        }
        {
         arriveTangent=-0.0109
         arriveTangentWeight=0.1035
         interpMode=RCIM_Cubic
         leaveTangent=-0.0109
         leaveTangentWeight=0.1035
         tangentMode=RCTM_User
         tangentWeightMode=RCTWM_WeightedBoth
         time=0.3256
         value=0
        }
        {
         arriveTangent=-0.0362
         arriveTangentWeight=0.088
         interpMode=RCIM_Cubic
         leaveTangent=-0.0362
         leaveTangentWeight=0.088
         tangentMode=RCTM_User
         tangentWeightMode=RCTWM_WeightedBoth
         time=0.5156
         value=0
        }
        <at 0.7015 0 RCIM_Cubic>
        {
         arriveTangentWeight=0.0831
         interpMode=RCIM_Cubic
         leaveTangentWeight=0.0831
         tangentMode=RCTM_User
         tangentWeightMode=RCTWM_WeightedBoth
         time=0.8534
         value=0
        }
        <at 1 0>
       ]}
      RedCurve {keys [
        {arriveTangent=2.3812, interpMode=RCIM_Cubic, leaveTangent=2.3812, tangentMode=RCTM_User, time=0, value=0}
        {
         arriveTangent=0.0019
         arriveTangentWeight=0.0857
         interpMode=RCIM_Cubic
         leaveTangent=0.0019
         leaveTangentWeight=0.0857
         tangentMode=RCTM_User
         tangentWeightMode=RCTWM_WeightedBoth
         time=0.0007
         value=0.1996
        }
        {
         arriveTangent=-0.0076
         arriveTangentWeight=0.0935
         interpMode=RCIM_Cubic
         leaveTangent=-0.0076
         leaveTangentWeight=0.0935
         tangentMode=RCTM_User
         tangentWeightMode=RCTWM_WeightedBoth
         time=0.1528
         value=-1.0376
        }
        {
         arriveTangent=0.0038
         arriveTangentWeight=0.1046
         interpMode=RCIM_Cubic
         leaveTangent=0.0038
         leaveTangentWeight=0.1046
         tangentMode=RCTM_User
         tangentWeightMode=RCTWM_WeightedBoth
         time=0.3011
         value=0.8898
        }
        {
         arriveTangent=0.1
         arriveTangentWeight=0.133
         interpMode=RCIM_Cubic
         leaveTangent=0.1
         leaveTangentWeight=0.133
         tangentMode=RCTM_User
         tangentWeightMode=RCTWM_WeightedBoth
         time=0.5
         value=-1.2818
        }
        {
         arriveTangent=0.0115
         arriveTangentWeight=0.1019
         interpMode=RCIM_Cubic
         leaveTangent=0.0115
         leaveTangentWeight=0.1019
         tangentMode=RCTM_User
         tangentWeightMode=RCTWM_WeightedBoth
         time=0.6726
         value=1.4154
        }
        {
         arriveTangent=0.011
         arriveTangentWeight=0.1017
         interpMode=RCIM_Cubic
         leaveTangent=0.011
         leaveTangentWeight=0.1017
         tangentMode=RCTM_User
         tangentWeightMode=RCTWM_WeightedBoth
         time=0.8289
         value=-0.9149
        }
        {
         arriveTangentWeight=0.0783
         interpMode=RCIM_Cubic
         leaveTangentWeight=0.0783
         tangentMode=RCTM_User
         tangentWeightMode=RCTWM_WeightedBoth
         time=0.9985
         value=0.2005
        }
       ]}
     }}
    Curve for Colors2=DataInterface() {propertyValues {
      AlphaCurve {keys [<at 0 0>]}
      BlueCurve {keys [
        {
         arriveTangentWeight=0.0518
         interpMode=RCIM_Cubic
         leaveTangentWeight=0.0518
         tangentMode=RCTM_User
         tangentWeightMode=RCTWM_WeightedBoth
         time=0.0031
         value=0
        }
        {
         arriveTangent=-0.0052
         arriveTangentWeight=0.1917
         interpMode=RCIM_Cubic
         leaveTangent=-0.0052
         leaveTangentWeight=0.1917
         tangentMode=RCTM_User
         tangentWeightMode=RCTWM_WeightedBoth
         time=0.1371
         value=0
        }
        {
         arriveTangent=-0.0293
         arriveTangentWeight=0.0977
         interpMode=RCIM_Cubic
         leaveTangent=-0.0293
         leaveTangentWeight=0.0977
         tangentMode=RCTM_User
         tangentWeightMode=RCTWM_WeightedBoth
         time=0.3191
         value=0
        }
        {
         arriveTangentWeight=0.1033
         interpMode=RCIM_Cubic
         leaveTangentWeight=0.1033
         tangentMode=RCTM_User
         tangentWeightMode=RCTWM_WeightedBoth
         time=0.5022
         value=0
        }
        {
         arriveTangentWeight=0.1042
         interpMode=RCIM_Cubic
         leaveTangentWeight=0.1042
         tangentMode=RCTM_User
         tangentWeightMode=RCTWM_WeightedBoth
         time=0.6685
         value=0
        }
        {
         arriveTangentWeight=0.0897
         interpMode=RCIM_Cubic
         leaveTangentWeight=0.0897
         tangentMode=RCTM_User
         tangentWeightMode=RCTWM_WeightedBoth
         time=0.8591
         value=0
        }
        {interpMode=RCIM_Cubic, tangentMode=RCTM_User, tangentWeightMode=RCTWM_WeightedBoth, time=1, value=0}
       ]}
      GreenCurve {keys [{interpMode=RCIM_Cubic, leaveTangentWeight=0.0491, tangentMode=RCTM_User, time=0, value=0}]}
      RedCurve {keys [
        {arriveTangent=-2.3812, interpMode=RCIM_Cubic, leaveTangent=-2.3812, leaveTangentWeight=0.0006, tangentMode=RCTM_User, time=0, value=0}
        {
         arriveTangent=-0.0019
         arriveTangentWeight=0.0857
         interpMode=RCIM_Cubic
         leaveTangent=-0.0019
         leaveTangentWeight=0.0857
         tangentMode=RCTM_User
         tangentWeightMode=RCTWM_WeightedBoth
         time=0.0007
         value=0
        }
       ]}
     }}
    Lerp Curve:NiagaraFloat=&FloatFromCurve(
     FloatCurve=DataInterface() {propertyValues {Curve {keys [<at 0 0 RCIM_Constant>,<at 0.1214 0 RCIM_Cubic>,<at 0.2959 0>]}}}
     CurveIndex=&ReturnNormalizedExecIndex(Normalized Index Scale=0)
    )
    NormalizeINdex:NiagaraFloat=&ReturnNormalizedExecIndex(Normalized Index Scale=0.65)
    Offset:NiagaraFloat=&SelectFloatFromArray(
     Direct Array Index=$Particles.PartitionParticles.Partition
     Float Selection Array=DataInterface() {propertyValues {FloatData [0,0.35,0.25,0.5,0.1,0.45,0.35,0.6]}}
    )
    STR:NiagaraFloat=&Multiply_Float(
     A=&Multiply_Float(A=20, B=&Lerp_Float(A=0.5, B=0.7, Alpha=$User.ParaGlideLeanAlpha))
     B=&FloatFromCurve(
      FloatCurve=DataInterface() {propertyValues {Curve {keys [
          <at 0 1 RCIM_Cubic>
          {arriveTangent=0.2683, interpMode=RCIM_Cubic, leaveTangent=0.2683, tangentMode=RCTM_User, time=0.2, value=1}
          {arriveTangent=-1.5254, interpMode=RCIM_Cubic, leaveTangent=-1.5254, tangentMode=RCTM_User, time=0.563, value=0.3}
          <at 0.6881 0>
         ]}}}
      CurveIndex=&ReturnNormalizedExecIndex()
     )
    )
    Speed=0.35
   ) {module=ScratchModule}
   &ScaleRibbonWidth(Ribbon Width Scale=&FloatFromCurve(
     FloatCurve=DataInterface() {propertyValues {Curve {keys [
         {arriveTangent=-0.0695, interpMode=RCIM_Cubic, leaveTangent=-0.0695, tangentMode=RCTM_User, time=0, value=1}
         {
          arriveTangent=0.0058
          arriveTangentWeight=0.4198
          interpMode=RCIM_Cubic
          leaveTangent=0.0058
          leaveTangentWeight=0.4198
          tangentMode=RCTM_User
          time=0.4966
          value=1
         }
         {
          arriveTangent=-0.0007
          arriveTangentWeight=0.7064
          interpMode=RCIM_Cubic
          leaveTangent=-0.0007
          leaveTangentWeight=0.7064
          tangentMode=RCTM_User
          time=0.8413
          value=0
         }
        ]}}}
     CurveIndex=&ReturnNormalizedExecIndex()
     Scale Curve=&Multiply_Float(
      A=&Multiply_Float(
       A=&SelectFloatFromArray(
        Direct Array Index=$Particles.PartitionParticles.Partition
        Float Selection Array=DataInterface() {propertyValues {FloatData [0.5,0.5,0.3,0.3,0.5,0.5,0.3,0.3]}}
       )
       B=&Multiply_Float(A=&Lerp_Float(A=1.2, B=0.9, Alpha=$System.FN_Contrail_Setup.Dive Value), B=$System.FEsizeMatch)
      )
      B=$System.FEsizeMatch
     )
    ))
   &ScaleColor(
    Scale Alpha=&Multiply_Float(
     A=&FloatFromCurve(
      FloatCurve=DataInterface() {propertyValues {Curve {
         defaultValue=1
         keys [
          {arriveTangent=0.2069, interpMode=RCIM_Cubic, leaveTangent=0.2069, tangentMode=RCTM_User, time=0, value=0}
          {arriveTangent=1.0911, interpMode=RCIM_Cubic, leaveTangent=1.0911, time=0.4392, value=0.3574}
          {tangentMode=RCTM_User, time=0.9165, value=1}
         ]
        }}}
      CurveIndex=&ReturnNormalizedExecIndex()
      Scale Curve=&FN_FrontEndFloatModifier(Gameplay=1, Front End=0.5)
     )
     B=&Lerp_Float(A=1.5, B=0.5, Alpha=$System.FN_Contrail_Setup.Dive Value)
    )
    Scale RGB=&VectorFromFloat(Value=&FNTimeOfDayScalar(Night Scale=0.5, Morning Scale=0.8, Evening Scale=0.8))
    ScaleA=true
    ScaleRGB=true
   )
   /SetVariables(Particles.RibbonID=&MakeNiagaraIDFromInt(Index:NiagaraInt32=$Particles.PartitionParticles.Partition, Acquire Tag:NiagaraInt32=$Particles.PartitionParticles.Partition)) {isSetParameters=true, module=SetVariables_#3}
   /SetVariables(Particles.ColorAdd:LinearColor=&Multiply_LinearColorByFloat(
     Linear Color=&SelectLinearColorFromArray(
      Array Sampling Mode=::NewEnumerator1
      Direct Array Index=$Particles.PartitionParticles.Partition
      Color Selection Array=DataInterface() {propertyValues {ColorData [
         <rgba 1 0.0642 0 1>
         <rgba 1 0.0642 0 1>
         <rgba 1 0.0642 0 1>
         <rgba 1 0.0642 0 1>
         <rgba 0 0.1611 1 0.5>
         <rgba 0 0.1611 1 0.5>
         <rgba 0 0.1611 1 0.5>
         <rgba 0 0.1611 1 0.5>
        ]}}
      Random Seed=?
     )
     Float=&Multiply_Float(
      A=&FloatFromCurve(
       FloatCurve=DataInterface() {propertyValues {Curve {
          defaultValue=1
          keys [
           {arriveTangent=0.2069, interpMode=RCIM_Cubic, leaveTangent=0.2069, tangentMode=RCTM_User, time=0, value=0}
           {arriveTangent=1, interpMode=RCIM_Cubic, leaveTangent=1, time=0.4336, value=0.6818}
           {tangentMode=RCTM_User, time=1, value=1}
          ]
         }}}
       CurveIndex=&ReturnNormalizedExecIndex()
       Scale Curve=&FN_FrontEndFloatModifier(Gameplay=1, Front End=0.6)
      )
      B=&Lerp_Float(A=1, B=0.3, Alpha=$System.FN_Contrail_Setup.Dive Value)
     )
    )) {isSetParameters=true, module=SetVariables_#4}
  ]
  rendererClasses [&NiagaraRibbonRendererProperties]
  renderers [
   {rendererClass=&NiagaraRibbonRendererProperties}
   {rendererClass=&NiagaraRibbonRendererProperties, rendererIndex=1}
  ]
 }
 Smoke_Tail {
  emitterUpdateScript [
   &EmitterState(Life Cycle Mode=::NewEnumerator1, Loop Behavior=::NewEnumerator1, Loop Delay=0.5, Loop Duration Mode=::NewEnumerator1, UseLoopDelay=true)
   &SpawnBurst_Instantaneous(Spawn Count=2, Use Spawn Probability=false, UseLoopCountLimit=false)
  ]
  particleSpawnScript [
   &PartitionParticles()
   &InitializeParticle(
    Color=&MaskLinearColor(
     Mask Source=1
     Mask Parameter=$Particles.PartitionParticles.Partition
     Mask=0
     Masked Linear Color=<rgba 1 0.14 0.0021 1>
     Passthrough Linear Color=<rgba 0 0.099 1 1>
    )
    Lifetime=2
    Mesh Scale=&FN_FrontEndVector3Modifier(Gameplay=<0.4 0.4 1>, Front End=<0.8 0.8 1>)
    Mesh Scale Mode=::NewEnumerator4
    Sprite Size Mode=::NewEnumerator3
    Uniform Sprite Size=50
    UsePositionOffset=false
    Write Material Random=true
   )
   &InitialMeshOrientation(
    Mesh Axis To Orient=<0 0 0>
    Mesh Orientation Mode=::NewEnumerator2
    Rotation=&Lerp_Vector(A=<0 0 0>, B=<0 0 0>, Alpha=&MakeFloatFromInt(Int=$Particles.PartitionParticles.Partition))
    Rotation Coordinate Space=::NewEnumerator0
    Use Rotation Vector=true
   )
   /SetVariables(Particles.WPO=25) {isSetParameters=true}
   &Color(Color=&MaskLinearColor(
     Mask Source=1
     Mask Parameter=$Particles.PartitionParticles.Partition
     Mask=0
     Masked Linear Color=<rgba 0.8438 0.0697 0 1>
     Passthrough Linear Color=<rgba 0.0048 0.087 0.6615 2>
    ))
  ]
  particleUpdateScript [
   &ParticleState(
    Kill Particles When Lifetime Has Elapsed=false
    Let Infinitely Lived Particles Die When Emitter Deactivates=true
    Loop Particles Lifetime=true
   )
   &SkeletalMeshLocation(
    Bone / Socket Index=&SelectIntFromArray(Direct Array Index=$Particles.PartitionParticles.Partition, Int Selection Array=DataInterface() {propertyValues {IntData [0,1]}})
    MaskSpawnGroup=false
    Mesh Sampling Type=::NewEnumerator5
    Orientation Sampling=::NewEnumerator1
    Sampled Position Offset=&FN_FrontEndVector3Modifier(Gameplay=<0 0 20>, Front End=<0 0 30>)
    Sampled Position Offset Coordinate Space:ENiagaraSkelMeshTransforms=::NewEnumerator2
    Skeletal Mesh=$System.Skeletal Mesh
    Socket Sampling Mode=::NewEnumerator5
   )
   /SetVariables(Particles.Position=&Add_Vector(
     A=&ConvertPositionToVector(Input Position=$Particles.Position, Convert To Absolute Large World Vector=false)
     B=&Add_Vector(A=&Multiply_VectorByFloat(Vector=$System.DiveVelocityScale, Float=10), B=<0 0 0>)
    )) {isSetParameters=true}
   &UpdateMeshOrientation(
    Facing Direction=$System.DiveVelocityScale
    Facing Interpolation=false
    Orientation Method=::NewEnumerator1
    Override Quat=false
    Stabilization Mode:ENiagara_OrientVectorCount=::NewEnumerator2
    TransformFacing=false
    TransformSide=false
   ) {module=UpdateMeshOrientation001}
   &UpdateMeshOrientation(Rotation Rate=-1, Rotation Vector=<0 0 1>) {enabled=false}
   &CameraOffset(Camera Offset Amount=8)
   &ScaleMeshSize(Scale Factor=&Multiply_Vector(
     A=&Lerp_Vector(A=<1 1 1.8>, B=<1 1 2>, Alpha=$System.Timeline.NormalizedTime)
     B=&MakeVector(
      X=1
      Y=1
      Z=&FloatFromCurve(
       FloatCurve=DataInterface() {propertyValues {Curve {keys [<at 0 0 RCIM_Cubic>, {arriveTangent=4.1531, interpMode=RCIM_Cubic, leaveTangent=4.1531, tangentMode=RCTM_User, time=0.5, value=1}]}}}
       CurveIndex=$Emitter.Age
      )
     )
    ))
   &ScaleMeshSize(Scale Factor=&VectorFromFloat(Value=$System.FN_Contrail_Setup.Dive Value)) {module=ScaleMeshSize001}
   &ScaleColor(
    Scale Alpha=&FloatFromCurve(
     FloatCurve=DataInterface() {propertyValues {Curve {keys [<at 0.5 0 RCIM_Cubic>, {arriveTangent=4.1531, interpMode=RCIM_Cubic, leaveTangent=4.1531, tangentMode=RCTM_User, time=1, value=1}]}}}
     CurveIndex=$Emitter.Age
     Scale Curve=&FN_FrontEndFloatModifier(Gameplay=0.6, Front End=0.3)
    )
    Scale RGB=&Multiply_VectorByFloat(Vector=&MakeVectorFromLinearColorRGB(Color=<rgba 1 1 1 1>), Float=1)
    ScaleA=true
    ScaleRGB=true
   ) {module=ScaleColor001}
   /SetVariables(Particles.Color_Yellow:LinearColor=&MakeLinearColorFromVectorAndFloat(
     "Vector (RGB)"=&Multiply_VectorByFloat(
      Vector=&MakeVectorFromLinearColorRGB(Color=&MaskLinearColor(
        Mask Source=1
        Mask Parameter=$Particles.PartitionParticles.Partition
        Mask=0
        Masked Linear Color=<rgba 1 0.14 0.0021 1>
        Passthrough Linear Color=<rgba 0 0.099 1 1>
       ))
      Float=1
     )
     "Float (Alpha)"=&FloatFromCurve(
      FloatCurve=DataInterface() {propertyValues {Curve {keys [<at 0 0 RCIM_Cubic>, {arriveTangent=2.7731, interpMode=RCIM_Cubic, leaveTangent=2.7731, tangentMode=RCTM_User, time=1, value=1}]}}}
      CurveIndex=$Emitter.Age
      Scale Curve=&Multiply_Float(A=0.25, B=$System.FN_Contrail_Setup.Dive Value)
     )
    )) {isSetParameters=true}
   /SetVariables(Particles.Scale_Yellow:Vector3f=&Multiply_Vector(A=$Particles.Scale, B=<0.75 0.75 0.8>)) {isSetParameters=true}
   &DynamicMaterialParameters(
    Index 0 Param 2=&RandomRangeFloat(Minimum=0, Maximum=5)
    Index 0 Param 3=&Multiply_Float(A=$Particles.WPO, B=&Lerp_Float(A=0.8, B=0.5, Alpha=$System.FN_Contrail_Setup.Dive Value))
    Index 2 Param 1=&Lerp_Float(A=0.2, B=1, Alpha=&Ceil_Float(Float=$System.FN_Contrail_Setup.Dive Value))
    Index 2 Param 2=&Lerp_Float(A=0, B=0.3, Alpha=&Ceil_Float(Float=$System.FN_Contrail_Setup.Dive Value))
    Module.Param0WriteEnabled=false
    Module.Param0WriteEnabled2=true
    Module.Param1WriteEnabled=true
    Module.Param1WriteEnabled2=true
    Module.Param2WriteEnabled=true
    Module.Param2WriteEnabled2=false
    Module.Param3WriteEnabled=false
    Module.Param3WriteEnabled2=false
    Write Parameter Index 2=true
   ) {module=DynamicMaterialParameters001}
   &UpdateMeshOrientation(
    Facing Direction=$System.FN_Contrail_Setup.Direction Vector
    Facing Interpolation=false
    Orientation Method=::NewEnumerator1
    Override Quat=false
    TransformFacing=false
    TransformUp=false
    Up Direction:Vector3f=<0 1 1>
   ) {enabled=false, module=UpdateMeshOrientation002}
  ]
  rendererClasses [&NiagaraMeshRendererProperties]
  renderers [
   {rendererClass=&NiagaraMeshRendererProperties, rendererIndex=0}
   {rendererClass=&NiagaraMeshRendererProperties, rendererIndex=1}
   {rendererClass=&NiagaraMeshRendererProperties, rendererIndex=2}
  ]
 }
 Sparks {
  emitterUpdateScript [
   &EmitterState()
   &SpawnRate(
    Spawn Probability=&Lerp_Float(A=1, B=0, Alpha=$System.FN_Contrail_Setup.Dive Value)
    SpawnRate=15
    Use Spawn Probability=true
   )
   &SpawnBurst_Instantaneous(
    Spawn Count=&Multiply_Int(A=0, B=&MakeIntFromFloat(Float=$System.FN_Contrail_Setup.Dive Value, Conversion Mode=::NewEnumerator2))
    Use Spawn Probability=false
    UseLoopCountLimit=false
   )
  ]
  particleSpawnScript [
   &PartitionParticles()
   &InitializeParticle(
    Color=&MaskLinearColor(
     Mask Source=1
     Mask Parameter=$Particles.PartitionParticles.Partition
     Mask=0
     Masked Linear Color=<rgba 2 0.2428 0 1>
     Passthrough Linear Color=<rgba 0 0.1979 2 1>
    )
    Lifetime Min=2
    Lifetime Mode=::NewEnumerator1
    Sprite Rotation Mode=::NewEnumerator2
    Sprite Size Max=<8 16>
    Sprite Size Min=<3 7>
    Sprite Size Mode=::NewEnumerator2
    UsePositionOffset=false
    Write Material Random=true
   )
   &SkeletalMeshLocation(
    Bone / Socket Index=$Particles.PartitionParticles.Partition
    MaskSpawnGroup=false
    Mesh Sampling Type=::NewEnumerator5
    Orientation Sampling=::NewEnumerator1
    Sampled Position Offset=&AddVectorToPosition(
     Position=&SelectVectorFromArray(
      Direct Array Index=$Particles.PartitionParticles.Partition
      Vector Selection Array=DataInterface() {propertyValues {FloatData [<0 0 0>, <0 0 0>]}}
     )
     Vector=<0 0 0>
    )
    Sampled Position Offset Coordinate Space:ENiagaraSkelMeshTransforms=::NewEnumerator2
    Skeletal Mesh=$System.Skeletal Mesh
    Socket Sampling Mode=::NewEnumerator5
   )
   &ShapeLocation(
    Cylinder Height=50
    Cylinder Height Midpoint=0
    Cylinder Radius=30
    MaskSpawnGroup=false
    Override Seed=false
    Shape Primitive=::NewEnumerator1
    Surface Only=false
    UseOwnerScale=true
   )
   &AddVelocity(Velocity=<0 0 100>)
  ]
  particleUpdateScript [
   &ParticleState()
   /Niagara/Modules/Update/Forces/LineAttractionForce(
    Attraction Strength=150
    Line End:Vector3f=&AddVectorToPosition(Position=<0 0 1000>, Vector=Linked() {linkedVariable {name=Particles.SkeletalMeshLocation.SampledPosition, type=NiagaraPosition}})
    Line Start:Vector3f=&AddVectorToPosition(Position=<0 0 1000>, Vector=Linked() {linkedVariable {name=Particles.SkeletalMeshLocation.SampledPosition, type=NiagaraPosition}})
    Use Attraction Falloff=false
   )
   /SetVariables(Particles.NewAge:NiagaraFloat=&Multiply_Float(A=$Particles.Age, B=1)) {enabled=false, isSetParameters=true}
   /SetVariables(Particles.Velocity=&Multiply_VectorByFloat(
     Vector=&Lerp_Vector(A=$System.FN_Contrail_Setup.Direction Vector, B=$System.FN_Contrail_Setup.Direction Velocity, Alpha=$Particles.NewAge)
     Float=0.2
    )) {enabled=false, isSetParameters=true}
   /SetVariables(Particles.Position=&AddVectorToPosition(
     Position=$Particles.Initial.Position
     Vector=&Multiply_VectorByFloat(Vector=$Particles.Velocity, Float=$Particles.Age)
    )) {enabled=false, isSetParameters=true}
   &ScaleSpriteSize(Scale Factor=&Vector2DFromFloat(Value=&FloatFromCurve(
      FloatCurve=DataInterface() {propertyValues {Curve {keys [
          <at 0 0 RCIM_Cubic>
          {interpMode=RCIM_Cubic, leaveTangent=-3.0296, tangentMode=RCTM_Break, time=0.2, value=1.2}
          {arriveTangent=-1.8107, interpMode=RCIM_Cubic, leaveTangent=-1.8107, tangentMode=RCTM_User, time=1, value=0}
         ]}}}
      Scale Curve=0.4
     ))) {module=ScaleSpriteSize001}
   /SetVariables(Particles.Glow03_Color=&MakeLinearColorFromVectorAndFloat(
     "Vector (RGB)"=&MakeVectorFromLinearColorRGB(Color=&MultiplyLinearColor(
       Linear Color=&MaskLinearColor(
        Mask Source=1
        Mask Parameter=$Particles.PartitionParticles.Partition
        Mask=0
        Masked Linear Color=<rgba 2 0.2428 0 1>
        Passthrough Linear Color=<rgba 0 0.1979 2 1>
       )
       Scale Factor=&MakeVector4_FromFloat(Float=2)
      ))
     "Float (Alpha)"=&Multiply_Float(
      A=&Multiply_Float(A=0.2, B=1)
      B=&FN_FrontEndFloatModifier(Gameplay=1, Front End=0.7)
     )
    )) {isSetParameters=true}
   /SetVariables(Particles.Glow03_Size=&Multiply_Vector2D(A=$Particles.SpriteSize, B=<3 2>)) {isSetParameters=true}
   &ScaleColor(Scale Alpha=&RandomRangeFloat(Minimum=0.2, Maximum=0.5), ScaleA=true, ScaleRGB=true)
   &Drag(Drag=1.5, Use Linear Drag=true, Use Rotational Drag=false)
   &VortexForce(
    Origin Pull Amount=200
    UseOriginOffset=true
    Vortex Force Amount=&Multiply_Float(
     A=400
     B=&FloatFromCurve(FloatCurve=DataInterface() {propertyValues {Curve {defaultValue=1, keys [<at 0 0 RCIM_Cubic>, {tangentMode=RCTM_User, time=0.5, value=1}]}}})
    )
    Vortex Origin:NiagaraPosition=$Particles.SkeletalMeshLocation.SampledPosition
    Vortex Origin Offset Coordinate Space:ENiagaraCoordinateSpace=::Local
   )
   &SolveForcesAndVelocity(Clamp Velocity=false, Limit Acceleration=false)
   &SpriteRotationRate(Rotation Rate=&Multiply_Float(
     A=600
     B=&RandomRangeFloat(
      Minimum=&RandomRangeFloat(Minimum=-1, Maximum=-0.5)
      Maximum=&RandomRangeFloat(Minimum=1, Maximum=0.5)
     )
    ))
  ]
  rendererClasses [&NiagaraSpriteRendererProperties]
  renderers [
   {rendererClass=&NiagaraSpriteRendererProperties}
   {rendererClass=&NiagaraSpriteRendererProperties, rendererIndex=1}
  ]
 }
 TrailLong {
  emitterUpdateScript [
   &EmitterState(
    Inactive Response=::NewEnumerator1
    Life Cycle Mode=::NewEnumerator1
    Loop Behavior=::NewEnumerator1
    Loop Duration Mode=::NewEnumerator1
    UseLoopDelay=false
   )
   &SpawnBurst_Instantaneous(Spawn Count=2, Use Spawn Probability=false, UseLoopCountLimit=false)
   /SetVariables(Emitter.Texture_Speed_Master:NiagaraFloat=&MakeCustomFloatFromBool(
     Bool=&SetBoolByFloatComparison(A=1, B=$User.ParaGlideLeanAlpha, Comparison Type=::NewEnumerator4)
     True Float=0.7
     False Float=&RandomRangeFloat(Minimum=0.4, Maximum=0.5)
    )) {isSetParameters=true}
   /SetVariables(Emitter.Texture_Speed_Master_Detail:NiagaraFloat=&MakeCustomFloatFromBool(
     Bool=&SetBoolByFloatComparison(A=1, B=$User.ParaGlideLeanAlpha, Comparison Type=::NewEnumerator4)
     True Float=1.5
     False Float=0.45
    )) {isSetParameters=true}
   /SetVariables(Emitter.SphereMaskU:NiagaraFloat=&Lerp_Float(A=1, B=0.95, Alpha=$User.ParaGlideLeanAlpha)) {isSetParameters=true}
  ]
  particleSpawnScript [
   &PartitionParticles()
   &InitializeParticle(
    Color=&MaskLinearColor(
     Mask Source=1
     Mask Parameter=$Particles.PartitionParticles.Partition
     Mask=0
     Masked Linear Color=<rgba 0.5 0.07 0.001 1>
     Passthrough Linear Color=<rgba 0 0.049 0.4948 1>
    )
    Mesh Scale=&SelectVectorFromArray(
     Direct Array Index=$Particles.PartitionParticles.Partition
     Vector Selection Array=DataInterface() {propertyValues {FloatData [<1 0.5 1>, <1 0.5 1>]}}
    )
    Mesh Scale Mode=::NewEnumerator4
    UsePositionOffset=true
    Write Material Random=true
   )
   &InitialMeshOrientation(
    Mesh Axis To Orient=<0 0 0>
    Mesh Orientation Mode=::NewEnumerator2
    Rotation=&Lerp_Vector(A=<0 0 0>, B=<0 0 0>, Alpha=&MakeFloatFromInt(Int=$Particles.PartitionParticles.Partition))
    Rotation Coordinate Space=::NewEnumerator0
    Use Rotation Vector=true
   )
  ]
  particleUpdateScript [
   &ParticleState(Kill Particles When Lifetime Has Elapsed=false, Let Infinitely Lived Particles Die When Emitter Deactivates=true)
   &ScaleMeshSize(Scale Factor=&VectorFromCurve(
     VectorCurve=DataInterface() {propertyValues {
       XCurve {defaultValue=0, keys [{tangentMode=RCTM_User, time=0, value=0}, {tangentMode=RCTM_User, time=1, value=1}]}
       YCurve {defaultValue=0, keys [{tangentMode=RCTM_User, time=0, value=0}, {tangentMode=RCTM_User, time=1, value=1}]}
       ZCurve {defaultValue=0, keys [{tangentMode=RCTM_User, time=0, value=0}, {tangentMode=RCTM_User, time=1, value=1}]}
      }}
     Scale Curve=&MakeCustomVectorFromBool(
      Bool=$User.bIsFrontEndPreview
      True Vector=&Lerp_Vector(A=<1 0.6 1>, B=<1 1.3 1>, Alpha=$User.ParaGlideLeanAlpha)
      False Vector=&Lerp_Vector(A=<1 0.9 1>, B=<1 1.5 1>, Alpha=$User.ParaGlideLeanAlpha)
     )
    ))
   &SkeletalMeshLocation(
    Bone / Socket Index=&SelectIntFromArray(Direct Array Index=$Particles.PartitionParticles.Partition, Int Selection Array=DataInterface() {propertyValues {IntData [0,1]}})
    MaskSpawnGroup=false
    Mesh Sampling Type=::NewEnumerator5
    Orientation Sampling=::NewEnumerator1
    Sampled Position Offset=<0 0 20>
    Sampled Position Offset Coordinate Space:ENiagaraSkelMeshTransforms=::NewEnumerator2
    Skeletal Mesh=$System.Skeletal Mesh
    Socket Sampling Mode=::NewEnumerator5
   )
   &DynamicMaterialParameters(
    Index 0 Param 1=&RandomRangeFloat(Minimum=0.2, Maximum=1)
    Index 0 Param 2=&RandomRangeFloat(Minimum=0.2, Maximum=1)
    Index 0 Param 3=&FN_FrontEndFloatModifier(Gameplay=0.5, Front End=0)
    Index 0 Param 4=20
    Index 3 Param 1=150
    Index 3 Param 2=0.3
    Index 3 Param 4=&FloatFromCurve(FloatCurve=DataInterface() {propertyValues {Curve {keys [<at 0 0>,<at 1 1>]}}})
    Module.Index0Param0Write=true
    Module.Index0Param1Write=true
    Module.Index0Param2Write=true
    Module.Index0Param3Write=true
    Module.Index2Param0Write=true
    Module.Index2Param1Write=true
    Module.Index2Param2Write=true
    Module.Index2Param3Write=true
    Module.Index3Param0Write=true
    Module.Index3Param1Write=true
    Module.Index3Param2Write=true
    Module.Index3Param3Write=true
    Write Parameter Index 2=true
    Write Parameter Index 3=true
   )
   &InterpolateOverTime(
    Rate Of Change:NiagaraFloat=&FN_FrontEndFloatModifier(Gameplay=10, Front End=10)
    Target Vector Value:Vector3f=&AddVectorToPosition(
     Position=$Particles.Position
     Vector=&TransformVector(
      Vector=&FN_FrontEndVector3Modifier(Gameplay=$System.IGCenterposition, Front End=$System.FECenterposition)
      Destination Space=::Simulation
     )
    )
    Variable Type=::NewEnumerator2
   )
   /SetVariables(Particles.Interpolate Dynamic=&MakeFloat4FromVectorAndFloat(
     "Vector (RGB)"=$StackContext.InterpolateOverTime.InterpolateOverTime_V3_.Moving Average
     "Float (Alpha)"=&FN_FrontEndFloatModifier(Gameplay=1, Front End=0)
    )) {isSetParameters=true}
   /SetVariables(Particles.Color002=&Multiply_LinearColorByFloat(
     Linear Color=&MultiplyLinearColor(
      Linear Color=&MakeCustomLinearColorFromBool(
       Bool=$User.bIsFrontEndPreview
       True Color=<rgba 0.0219 0.0346 0.0729 1>
       False Color=&MaskLinearColor(
        Mask Source=1
        Mask Parameter=$Particles.PartitionParticles.Partition
        Mask=0
        Masked Linear Color=<rgba 0.4583 0.0642 0 1>
        Passthrough Linear Color=<rgba 0 0.0304 0.3073 1>
       )
      )
      Scale Factor=&MakeVector4_FromLinearColor(LinearColor=&FNTimeOfDayColor(Night Color=<rgba 0.1927 0.3945 1 1>, Morning Color=<rgba 0.6 0.6 0.6 1>, Day Color=<rgba 0.8 0.8 0.8 1>, Evening Color=<rgba 0.6 0.6 0.6 1>))
     )
     Float=&Multiply_Float(
      A=&FN_FrontEndFloatModifier(Gameplay=1, Front End=0.6)
      B=&FN_PickFloatDependOnPlatform(TargetPlatformActive=1, TargetPlatformNoActive=0.7)
     )
    )) {enabled=false, isSetParameters=true}
   /SetVariables(Particles.Color002=&MultiplyLinearColor(
     Linear Color=$Particles.Color002
     Scale Factor=&MakeVector4(X=5, Y=5, Z=5, W=$System.FN_Contrail_Setup.Dive Value)
    )) {enabled=false, isSetParameters=true}
   /SetVariables(Particles.Size_Detail:Vector3f=&VectorFromCurve(
     VectorCurve=DataInterface() {propertyValues {
       XCurve {defaultValue=1, keys [{tangentMode=RCTM_User, time=0, value=0}, {tangentMode=RCTM_User, time=1, value=1}]}
       YCurve {defaultValue=1, keys [{tangentMode=RCTM_User, time=0, value=0}, {tangentMode=RCTM_User, time=1, value=1}]}
       ZCurve {defaultValue=1, keys [{tangentMode=RCTM_User, time=0, value=0}, {tangentMode=RCTM_User, time=1, value=1}]}
      }}
     Scale Curve=<5 0.5 5>
    )) {isSetParameters=true}
   &ScaleColor(
    Scale Alpha=&FloatFromCurve(
     FloatCurve=DataInterface() {propertyValues {Curve {
        defaultValue=0
        keys [
         {arriveTangent=4.1785, interpMode=RCIM_Cubic, leaveTangent=4.1785, tangentMode=RCTM_User, time=0, value=0}
         {arriveTangent=1.1179, interpMode=RCIM_Cubic, leaveTangent=1.1179, tangentMode=RCTM_User, time=0.2042, value=0.6228}
         {arriveTangent=0.0224, interpMode=RCIM_Cubic, leaveTangent=0.0224, tangentMode=RCTM_User, time=1, value=1}
        ]
       }}}
     CurveIndex=$Emitter.Age
     Scale Curve=&FN_FrontEndFloatModifier(Gameplay=1, Front End=0.5)
    )
    Scale RGB=&Multiply_Vector(
     A=&MakeVectorFromLinearColorRGB(Color=&FNTimeOfDayColor(Night Color=<rgba 0.5 0.625 1 1>, Morning Color=<rgba 1 1 1 1>, Day Color=<rgba 1 1 1 1>, Evening Color=<rgba 1 1 1 1>))
     B=&VectorFromFloat(Value=&FN_FrontEndFloatModifier(Gameplay=1.2, Front End=1))
    )
    ScaleA=true
    ScaleRGB=true
   )
   &ScaleColor(Scale Alpha=$System.FN_Contrail_Setup.Dive Value, ScaleA=true, ScaleRGB=true) {module=ScaleColor002}
   &UpdateMeshOrientation(
    Facing Direction=$System.FN_Contrail_Setup.Direction Vector
    Facing Interpolation=false
    Orientation Method=::NewEnumerator1
    Override Quat=false
    TransformFacing=false
    TransformUp=false
   ) {enabled=false}
   &ScaleMeshSize(Scale Factor=&VectorFromFloat(Value=&Ceil_Float(Float=$System.FN_Contrail_Setup.Dive Value))) {module=ScaleMeshSize001}
  ]
  rendererClasses [&NiagaraMeshRendererProperties]
  renderers [{rendererClass=&NiagaraMeshRendererProperties, rendererIndex=0}]
 }
}
summary {
 systemName=NS_Contrail_Mythical
 emitters [
  {
   bEnabled=false
   emitterName=Mesh_SkeletalMesh
   rendererClasses [&NiagaraMeshRendererProperties, &NiagaraSpriteRendererProperties]
  }
  {
   bEnabled=false
   emitterName=Particle_Root
   rendererClasses [&NiagaraSpriteRendererProperties, &NiagaraMeshRendererProperties]
  }
  {bEnabled=false, emitterName=Particles_SkeletalMesh, rendererClasses [&NiagaraSpriteRendererProperties]}
  {bEnabled=false, emitterName=Gizmo_LocalSpace, rendererClasses [&NiagaraMeshRendererProperties]}
  {
   bEnabled=false
   emitterName=ParticletoRibbon_SkeletalMesh
   rendererClasses [&NiagaraSpriteRendererProperties, &NiagaraRibbonRendererProperties]
  }
  {
   bEnabled=false
   emitterName=FN_Contrails_Kite_v2
   rendererClasses [&NiagaraMeshRendererProperties, &NiagaraRibbonRendererProperties]
  }
  {emitterName=TrailLong, rendererClasses [&NiagaraMeshRendererProperties]}
  {emitterName=BodyTrails, rendererClasses [&NiagaraMeshRendererProperties]}
  {emitterName=Smoke_Tail, rendererClasses [&NiagaraMeshRendererProperties]}
  {emitterName=Sparks, rendererClasses [&NiagaraSpriteRendererProperties]}
  {emitterName=Main001, rendererClasses [&NiagaraRibbonRendererProperties]}
  {emitterName=FinTrail002, rendererClasses [&NiagaraRibbonRendererProperties]}
  {emitterName=SideFins001, rendererClasses [&NiagaraRibbonRendererProperties]}
 ]
 userVariables [
  User.bIsFrontEnd {type=NiagaraBool}
  User.bIsFrontEndPreview {type=NiagaraBool}
  User.ColorA {type=LinearColor, description="", defaultValue {struct=LinearColor, value=<rgba 0.5559 0.6633 0.9583 1>}}
  User.ColorB {type=LinearColor, description="", defaultValue {struct=LinearColor, value=<rgba 0.7969 0.0266 0 1>}}
  User.ParaGlideLeanAlpha {type=NiagaraFloat, defaultValue=0, description=""}
  User.RibbonIDOffset {type=NiagaraInt32}
  User.SkeletalMesh {
   type=NiagaraDataInterfaceSkeletalMesh
   defaultValue {
    struct=NiagaraExt_VariableValue_DataInterface
    value {dataInterface=&self:NiagaraDataInterfaceSkeletalMesh_0}
   }
  }
 ]
}
